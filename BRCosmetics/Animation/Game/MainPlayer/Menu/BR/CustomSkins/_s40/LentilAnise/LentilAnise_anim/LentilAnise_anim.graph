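# Level Sequence for authoring a main-player menu animation for the LentilAnise custom skin in Battle Royale cosmetics.
# One spawnable FortPlayerMannequin is keyed through a Fortnite_Biped_CtrlRig track, with static IK foot controls and subtle FK arm motion.
# A second spawnable mannequin stands offset and rotated beside it, playing a skeletal animation over roughly the same frame range.
# Visible animated channels start and end on matching values, consistent with a looping idle.
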LevelSequence /BRCosmetics/Animation/Game/MainPlayer/Menu/BR/CustomSkins/_s40/LentilAnise/LentilAnise_anim/LentilAnise_anim
bindings [
 mannequin3 {
  id=#1
  possessed_class=FortPlayerMannequin
  tracks [
   {type=MovieSceneSpawnTrack, sections [{type=MovieSceneSpawnSection, channels [None {type=MovieSceneScriptingBoolChannel, default=true}]}]}
   {
    type=MovieScene3DTransformTrack
    sections [{
      type=MovieScene3DTransformSection
      channels [
       Location.X {type=MovieSceneScriptingDoubleChannel, default=87.6392}
       Location.Y {type=MovieSceneScriptingDoubleChannel, default=0}
       Location.Z {type=MovieSceneScriptingDoubleChannel, default=0}
       Rotation.X {type=MovieSceneScriptingDoubleChannel, default=0}
       Rotation.Y {type=MovieSceneScriptingDoubleChannel, default=0}
       Rotation.Z {type=MovieSceneScriptingDoubleChannel, default=-80.4843}
       Scale.X {type=MovieSceneScriptingDoubleChannel, default=1}
       Scale.Y {type=MovieSceneScriptingDoubleChannel, default=1}
       Scale.Z {type=MovieSceneScriptingDoubleChannel, default=1}
      ]
     }]
   }
   {type=MovieSceneSkeletalAnimationTrack, sections [{type=MovieSceneSkeletalAnimationSection, end_frame=328, start_frame=-1}]}
  ]
 }
 mannequin {
  id=#2
  possessed_class=FortPlayerMannequin
  tracks [
   {type=MovieSceneSpawnTrack, sections [{type=MovieSceneSpawnSection, channels [None {type=MovieSceneScriptingBoolChannel, default=true}]}]}
   {
    type=MovieScene3DTransformTrack
    sections [{
      type=MovieScene3DTransformSection
      channels [
       Location.X {type=MovieSceneScriptingDoubleChannel, default=0}
       Location.Y {type=MovieSceneScriptingDoubleChannel, default=0}
       Location.Z {type=MovieSceneScriptingDoubleChannel, default=0}
       Rotation.X {type=MovieSceneScriptingDoubleChannel, default=0}
       Rotation.Y {type=MovieSceneScriptingDoubleChannel, default=0}
       Rotation.Z {type=MovieSceneScriptingDoubleChannel, default=0}
       Scale.X {type=MovieSceneScriptingDoubleChannel, default=1}
       Scale.Y {type=MovieSceneScriptingDoubleChannel, default=1}
       Scale.Z {type=MovieSceneScriptingDoubleChannel, default=1}
      ]
     }]
   }
   {
    type=MovieSceneControlRigParameterTrack
    display_name=Fortnite_Biped_CtrlRig
    sections [{
      type=MovieSceneControlRigParameterSection
      channels [
       leg_l_stretch_switch {type=MovieSceneScriptingBoolChannel, default=false, const [true,50]}
       leg_r_stretch_switch {type=MovieSceneScriptingBoolChannel, default=false, const [true,50]}
       arm_l_fk_local {type=MovieSceneScriptingBoolChannel, default=true, const [false,51]}
       Auto Interpolate_3 {type=MovieSceneScriptingBoolChannel, default=false, const [false,50]}
       Auto Interpolate {type=MovieSceneScriptingBoolChannel, default=false, const [false,50]}
       arm_l_stretch_switch {type=MovieSceneScriptingBoolChannel, default=false, const [false,50]}
       Keep Aligned {type=MovieSceneScriptingBoolChannel, default=false, const [false,50]}
       arm_r_fk_local {type=MovieSceneScriptingBoolChannel, default=true, const [false,50]}
       Auto Interpolate_4 {type=MovieSceneScriptingBoolChannel, default=false, const [false,50]}
       Auto Interpolate_2 {type=MovieSceneScriptingBoolChannel, default=false, const [false,50]}
       arm_r_stretch_switch {type=MovieSceneScriptingBoolChannel, default=false, const [false,50]}
       Keep Aligned_2 {type=MovieSceneScriptingBoolChannel, default=false, const [false,50]}
       head_local {type=MovieSceneScriptingBoolChannel, default=true, const [true,50]}
       Stretch {type=MovieSceneScriptingBoolChannel, default=false, const [false,50]}
       head_stretch {type=MovieSceneScriptingBoolChannel, default=false, const [false,50]}
       spine_stretch_switch {type=MovieSceneScriptingBoolChannel, default=false, const [false,50]}
       arm_l_fk_ik_switch {type=MovieSceneScriptingBoolChannel, default=false, const [false,50]}
       leg_l_fk_ik_switch {type=MovieSceneScriptingBoolChannel, default=true, const [true,50]}
       arm_r_fk_ik_switch {type=MovieSceneScriptingBoolChannel, default=false, const [false,50]}
       leg_r_fk_ik_switch {type=MovieSceneScriptingBoolChannel, default=true, const [true,50]}
       ShowBodyControls {type=MovieSceneScriptingBoolChannel, default=true, const [true,50]}
       spine_fk_ik_switch {type=MovieSceneScriptingBoolChannel, default=false, const [false,50]}
       neck_fk_ik_switch {type=MovieSceneScriptingBoolChannel, default=false, const [false,50]}
       space {type=MovieSceneScriptingByteChannel, default=1, const [2,50]}
       global_ctrl.Location.X {type=MovieSceneScriptingFloatChannel, default=0, const [0,50]}
       global_ctrl.Location.Y {type=MovieSceneScriptingFloatChannel, default=0, const [0,50]}
       global_ctrl.Location.Z {type=MovieSceneScriptingFloatChannel, default=0, const [0,50]}
       global_ctrl.Rotation.X {type=MovieSceneScriptingFloatChannel, default=0, const [0,50]}
       global_ctrl.Rotation.Y {type=MovieSceneScriptingFloatChannel, default=0, const [0,50]}
       global_ctrl.Rotation.Z {type=MovieSceneScriptingFloatChannel, default=0, const [0,50]}
       root_ctrl.Location.X {type=MovieSceneScriptingFloatChannel, default=0, const [0,50]}
       root_ctrl.Location.Y {type=MovieSceneScriptingFloatChannel, default=0, const [0,50]}
       root_ctrl.Location.Z {type=MovieSceneScriptingFloatChannel, default=0, const [0,50]}
       root_ctrl.Rotation.X {type=MovieSceneScriptingFloatChannel, default=0, const [0,50]}
       root_ctrl.Rotation.Y {type=MovieSceneScriptingFloatChannel, default=0, const [0,50]}
       root_ctrl.Rotation.Z {type=MovieSceneScriptingFloatChannel, default=0, const [0,50]}
       root_ctrl.Scale.X {type=MovieSceneScriptingFloatChannel, default=1, const [1,50]}
       root_ctrl.Scale.Y {type=MovieSceneScriptingFloatChannel, default=1, const [1,50]}
       root_ctrl.Scale.Z {type=MovieSceneScriptingFloatChannel, default=1, const [1,50]}
       body_offset_ctrl.Location.X {type=MovieSceneScriptingFloatChannel, default=0, const [0,50]}
       body_offset_ctrl.Location.Y {type=MovieSceneScriptingFloatChannel, default=0, const [0,50]}
       body_offset_ctrl.Location.Z {type=MovieSceneScriptingFloatChannel, default=0, const [0,50]}
       body_offset_ctrl.Rotation.X {type=MovieSceneScriptingFloatChannel, default=0, const [0,50]}
       body_offset_ctrl.Rotation.Y {type=MovieSceneScriptingFloatChannel, default=0, const [0,50]}
       body_offset_ctrl.Rotation.Z {type=MovieSceneScriptingFloatChannel, default=0, const [0,50]}
       body_offset_ctrl.Scale.X {type=MovieSceneScriptingFloatChannel, default=1, const [1,50]}
       body_offset_ctrl.Scale.Y {type=MovieSceneScriptingFloatChannel, default=1, const [1,50]}
       body_offset_ctrl.Scale.Z {type=MovieSceneScriptingFloatChannel, default=1, const [1,50]}
       thigh_l_fk_ctrl.Location.X {type=MovieSceneScriptingFloatChannel, default=0, thinned_from=50, keys [[0,0],[328,0]]}
       thigh_l_fk_ctrl.Location.Y {type=MovieSceneScriptingFloatChannel, default=0, thinned_from=50, keys [[0,0],[328,0]]}
       thigh_l_fk_ctrl.Location.Z {type=MovieSceneScriptingFloatChannel, default=0, thinned_from=50, keys [[0,0],[328,0]]}
       thigh_l_fk_ctrl.Rotation.X {
        type=MovieSceneScriptingFloatChannel
        default=1.0553
        thinned_from=50
        keys [[0,55.2583],[14,55.2583],[26,55.4104],[118,55.2583],[126,54.9543],[140,54.8643],[151,55.173],[159,55.2888],[328,55.2583]]
       }
       thigh_l_fk_ctrl.Rotation.Y {
        type=MovieSceneScriptingFloatChannel
        default=-1.0258
        thinned_from=50
        keys [[0,1.7001],[14,1.7001],[26,1.6905],[35,1.6988],[118,1.7001],[122,1.5172],[126,1.3237],[130,1.2541],[134,1.2227],[140,1.2123],[145,1.5073],[151,1.999],[159,2.294],[184,2.4461],[192,2.471],[238,2.471],[244,2.3987],[249,2.2783],[254,1.9169],[260,1.7814],[266,1.7306],[273,1.7115],[281,1.7001],[328,1.7001]]
       }
       thigh_l_fk_ctrl.Rotation.Z {
        type=MovieSceneScriptingFloatChannel
        default=-0.0727
        thinned_from=50
        keys [[0,1.1551],[14,1.1551],[26,1.0672],[32,1.1221],[35,1.1427],[39,1.1506],[53,1.1551],[118,1.1551],[126,1.1097],[130,1.1014],[134,1.0976],[140,1.0963],[145,1.1354],[151,1.2005],[159,1.2395],[184,1.2596],[192,1.2629],[238,1.2629],[244,1.2528],[249,1.236],[254,1.1854],[260,1.1665],[266,1.1594],[281,1.1551],[328,1.1551]]
       }
       thigh_l_fk_ctrl.Scale.X {type=MovieSceneScriptingFloatChannel, default=1, const [1,50]}
       thigh_l_fk_ctrl.Scale.Y {type=MovieSceneScriptingFloatChannel, default=1, const [1,50]}
       thigh_l_fk_ctrl.Scale.Z {type=MovieSceneScriptingFloatChannel, default=1, const [1,50]}
       calf_l_fk_ctrl.Location.X {
        type=MovieSceneScriptingFloatChannel
        default=0
        thinned_from=50
        keys [[0,-1.1732],[14,-1.1732],[26,-1.2334],[32,-1.1958],[35,-1.1817],[39,-1.1763],[60,-1.1732],[118,-1.1732],[126,-1.217],[130,-1.2252],[134,-1.2288],[140,-1.23],[145,-1.2201],[151,-1.2035],[159,-1.1936],[244,-1.1863],[249,-1.184],[254,-1.1773],[328,-1.1732]]
       }
       calf_l_fk_ctrl.Location.Y {
        type=MovieSceneScriptingFloatChannel
        default=0
        thinned_from=50
        keys [[0,0.1537],[14,0.1537],[26,0.1616],[32,0.1567],[35,0.1548],[39,0.1541],[60,0.1537],[118,0.1537],[126,0.1594],[130,0.1605],[134,0.161],[140,0.1611],[145,0.1598],[151,0.1577],[159,0.1564],[244,0.1554],[249,0.1551],[254,0.1542],[328,0.1537]]
       }
       calf_l_fk_ctrl.Location.Z {
        type=MovieSceneScriptingFloatChannel
        default=0
        thinned_from=50
        keys [[0,-0.0218],[14,-0.0218],[26,-0.0229],[35,-0.022],[118,-0.0218],[126,-0.0226],[140,-0.0228],[151,-0.0224],[159,-0.0222],[249,-0.022],[328,-0.0218]]
       }
       calf_l_fk_ctrl.Rotation.X {type=MovieSceneScriptingFloatChannel, default=0.0724, thinned_from=50, keys [[0,-0.0692],[328,-0.0692]]}
       calf_l_fk_ctrl.Rotation.Y {type=MovieSceneScriptingFloatChannel, default=1.0549, thinned_from=50, keys [[0,1.0551],[328,1.0551]]}
       calf_l_fk_ctrl.Rotation.Z {type=MovieSceneScriptingFloatChannel, default=0.2163, const [-7.4634,50]}
       calf_l_fk_ctrl.Scale.X {type=MovieSceneScriptingFloatChannel, default=1, const [1,50]}
       calf_l_fk_ctrl.Scale.Y {type=MovieSceneScriptingFloatChannel, default=1, const [1,50]}
       calf_l_fk_ctrl.Scale.Z {type=MovieSceneScriptingFloatChannel, default=1, const [1,50]}
       foot_l_fk_ctrl.Location.X {
        type=MovieSceneScriptingFloatChannel
        default=0
        thinned_from=50
        keys [[0,-1.3022],[14,-1.3022],[26,-1.369],[32,-1.3273],[35,-1.3116],[39,-1.3056],[60,-1.3022],[118,-1.3022],[126,-1.3509],[130,-1.3598],[134,-1.3639],[140,-1.3652],[145,-1.3542],[151,-1.3359],[159,-1.3248],[244,-1.3167],[249,-1.3142],[254,-1.3067],[328,-1.3022]]
       }
       foot_l_fk_ctrl.Location.Y {type=MovieSceneScriptingFloatChannel, default=0.0627, thinned_from=50, keys [[0,0.0627],[328,0.0627]]}
       foot_l_fk_ctrl.Location.Z {type=MovieSceneScriptingFloatChannel, default=-0.8453, const [-0.8453,50]}
       foot_l_fk_ctrl.Rotation.X {
        type=MovieSceneScriptingFloatChannel
        default=-0.9801
        thinned_from=50
        keys [[0,-2.2386],[140,-2.2393],[145,-2.2554],[151,-2.2822],[159,-2.2982],[238,-2.3079],[244,-2.3014],[249,-2.2905],[254,-2.2581],[260,-2.2459],[328,-2.2386]]
       }
       foot_l_fk_ctrl.Rotation.Y {
        type=MovieSceneScriptingFloatChannel
        default=0.1001
        thinned_from=50
        keys [[0,-2.9642],[140,-2.9623],[145,-3.0151],[151,-3.103],[159,-3.1557],[184,-3.1829],[238,-3.1874],[244,-3.1665],[249,-3.1316],[254,-3.0269],[260,-2.9877],[266,-2.973],[281,-2.9642],[328,-2.9642]]
       }
       foot_l_fk_ctrl.Rotation.Z {
        type=MovieSceneScriptingFloatChannel
        default=-0.1441
        thinned_from=50
        keys [[0,12.0468],[140,12.0553],[151,12.1546],[159,12.1918],[238,12.2141],[249,12.1723],[254,12.0939],[266,12.0534],[328,12.0468]]
       }
       foot_l_fk_ctrl.Scale.X {type=MovieSceneScriptingFloatChannel, default=1, const [1,50]}
       foot_l_fk_ctrl.Scale.Y {type=MovieSceneScriptingFloatChannel, default=1, const [1,50]}
       foot_l_fk_ctrl.Scale.Z {type=MovieSceneScriptingFloatChannel, default=1, const [1,50]}
       ball_l_fk_ctrl.Location.X {type=MovieSceneScriptingFloatChannel, default=0, const [0,50]}
       ball_l_fk_ctrl.Location.Y {type=MovieSceneScriptingFloatChannel, default=0, const [0,50]}
       ball_l_fk_ctrl.Location.Z {type=MovieSceneScriptingFloatChannel, default=0, const [0,50]}
       ball_l_fk_ctrl.Rotation.X {type=MovieSceneScriptingFloatChannel, default=0, const [0,50]}
       ball_l_fk_ctrl.Rotation.Y {type=MovieSceneScriptingFloatChannel, default=0, const [0,50]}
       ball_l_fk_ctrl.Rotation.Z {type=MovieSceneScriptingFloatChannel, default=0, const [0,50]}
       ball_l_fk_ctrl.Scale.X {type=MovieSceneScriptingFloatChannel, default=1, const [1,50]}
       ball_l_fk_ctrl.Scale.Y {type=MovieSceneScriptingFloatChannel, default=1, const [1,50]}
       ball_l_fk_ctrl.Scale.Z {type=MovieSceneScriptingFloatChannel, default=1, const [1,50]}
       foot_l_ik_ctrl.Location.X {type=MovieSceneScriptingFloatChannel, default=0, const [6.5639,50]}
       foot_l_ik_ctrl.Location.Y {type=MovieSceneScriptingFloatChannel, default=0, const [4.2963,50]}
       foot_l_ik_ctrl.Location.Z {type=MovieSceneScriptingFloatChannel, default=0, const [0,50]}
       foot_l_ik_ctrl.Rotation.X {type=MovieSceneScriptingFloatChannel, default=0, const [0,50]}
       foot_l_ik_ctrl.Rotation.Y {type=MovieSceneScriptingFloatChannel, default=0, const [0,50]}
       foot_l_ik_ctrl.Rotation.Z {type=MovieSceneScriptingFloatChannel, default=0, const [-29.5,50]}
       foot_l_ik_ctrl.Scale.X {type=MovieSceneScriptingFloatChannel, default=1, const [1,50]}
       foot_l_ik_ctrl.Scale.Y {type=MovieSceneScriptingFloatChannel, default=1, const [1,50]}
       foot_l_ik_ctrl.Scale.Z {type=MovieSceneScriptingFloatChannel, default=1, const [1,50]}
       ball_l_ik_ctrl.Location.X {type=MovieSceneScriptingFloatChannel, default=0, const [0,50]}
       ball_l_ik_ctrl.Location.Y {type=MovieSceneScriptingFloatChannel, default=0, const [0,50]}
       ball_l_ik_ctrl.Location.Z {type=MovieSceneScriptingFloatChannel, default=0, const [0,50]}
       ball_l_ik_ctrl.Rotation.X {type=MovieSceneScriptingFloatChannel, default=0, const [0,50]}
       ball_l_ik_ctrl.Rotation.Y {type=MovieSceneScriptingFloatChannel, default=0, const [0,50]}
       ball_l_ik_ctrl.Rotation.Z {type=MovieSceneScriptingFloatChannel, default=0, const [0,50]}
       ball_l_ik_ctrl.Scale.X {type=MovieSceneScriptingFloatChannel, default=1, const [1,50]}
       ball_l_ik_ctrl.Scale.Y {type=MovieSceneScriptingFloatChannel, default=1, const [1,50]}
       ball_l_ik_ctrl.Scale.Z {type=MovieSceneScriptingFloatChannel, default=1, const [1,50]}
       foot_bk1_l_ctrl.Location.X {type=MovieSceneScriptingFloatChannel, default=0, const [0,50]}
       foot_bk1_l_ctrl.Location.Y {type=MovieSceneScriptingFloatChannel, default=0, const [0,50]}
       foot_bk1_l_ctrl.Location.Z {type=MovieSceneScriptingFloatChannel, default=0, const [0,50]}
       foot_bk1_l_ctrl.Rotation.X {type=MovieSceneScriptingFloatChannel, default=0, const [0,50]}
       foot_bk1_l_ctrl.Rotation.Y {type=MovieSceneScriptingFloatChannel, default=0, const [0,50]}
       foot_bk1_l_ctrl.Rotation.Z {type=MovieSceneScriptingFloatChannel, default=0, const [0,50]}
       foot_bk1_l_ctrl.Scale.X {type=MovieSceneScriptingFloatChannel, default=1, const [1,50]}
       foot_bk1_l_ctrl.Scale.Y {type=MovieSceneScriptingFloatChannel, default=1, const [1,50]}
       foot_bk1_l_ctrl.Scale.Z {type=MovieSceneScriptingFloatChannel, default=1, const [1,50]}
       foot_roll_l_ctrl.Location.X {type=MovieSceneScriptingFloatChannel, default=0, const [0,50]}
       foot_roll_l_ctrl.Location.Y {type=MovieSceneScriptingFloatChannel, default=0, const [0,50]}
       foot_roll_l_ctrl.Location.Z {type=MovieSceneScriptingFloatChannel, default=0, const [0,50]}
       foot_roll_l_ctrl.Rotation.X {type=MovieSceneScriptingFloatChannel, default=0, const [0,50]}
       foot_roll_l_ctrl.Rotation.Y {type=MovieSceneScriptingFloatChannel, default=0, const [0,50]}
       foot_roll_l_ctrl.Rotation.Z {type=MovieSceneScriptingFloatChannel, default=0, const [0,50]}
       foot_roll_l_ctrl.Scale.X {type=MovieSceneScriptingFloatChannel, default=1, const [1,50]}
       foot_roll_l_ctrl.Scale.Y {type=MovieSceneScriptingFloatChannel, default=1, const [1,50]}
       foot_roll_l_ctrl.Scale.Z {type=MovieSceneScriptingFloatChannel, default=1, const [1,50]}
       roll_blend_l_ctrl {type=MovieSceneScriptingFloatChannel, default=1, const [1,50]}
       tip_l_ctrl.Location.X {type=MovieSceneScriptingFloatChannel, default=0, const [0,50]}
       tip_l_ctrl.Location.Y {type=MovieSceneScriptingFloatChannel, default=0, const [0,50]}
       tip_l_ctrl.Location.Z {type=MovieSceneScriptingFloatChannel, default=0, const [0,50]}
       tip_l_ctrl.Rotation.X {type=MovieSceneScriptingFloatChannel, default=0, const [0,50]}
       tip_l_ctrl.Rotation.Y {type=MovieSceneScriptingFloatChannel, default=0, const [0,50]}
       tip_l_ctrl.Rotation.Z {type=MovieSceneScriptingFloatChannel, default=0, const [0,50]}
       tip_l_ctrl.Scale.X {type=MovieSceneScriptingFloatChannel, default=1, const [1,50]}
       tip_l_ctrl.Scale.Y {type=MovieSceneScriptingFloatChannel, default=1, const [1,50]}
       tip_l_ctrl.Scale.Z {type=MovieSceneScriptingFloatChannel, default=1, const [1,50]}
       heel_l_ctrl.Location.X {type=MovieSceneScriptingFloatChannel, default=0, const [0,50]}
       heel_l_ctrl.Location.Y {type=MovieSceneScriptingFloatChannel, default=0, const [0,50]}
       heel_l_ctrl.Location.Z {type=MovieSceneScriptingFloatChannel, default=0, const [0,50]}
       heel_l_ctrl.Rotation.X {type=MovieSceneScriptingFloatChannel, default=0, const [0,50]}
       heel_l_ctrl.Rotation.Y {type=MovieSceneScriptingFloatChannel, default=0, const [0,50]}
       heel_l_ctrl.Rotation.Z {type=MovieSceneScriptingFloatChannel, default=0, const [0,50]}
       heel_l_ctrl.Scale.X {type=MovieSceneScriptingFloatChannel, default=1, const [1,50]}
       heel_l_ctrl.Scale.Y {type=MovieSceneScriptingFloatChannel, default=1, const [1,50]}
       heel_l_ctrl.Scale.Z {type=MovieSceneScriptingFloatChannel, default=1, const [1,50]}
       leg_l_softness {type=MovieSceneScriptingFloatChannel, default=0, const [0,50]}
       leg_l_pv_ik_ctrl.Location.X {type=MovieSceneScriptingFloatChannel, default=0, const [0,50]}
       leg_l_pv_ik_ctrl.Location.Y {type=MovieSceneScriptingFloatChannel, default=0, const [0,50]}
       leg_l_pv_ik_ctrl.Location.Z {type=MovieSceneScriptingFloatChannel, default=0, const [0,50]}
       leg_l_pv_ik_ctrl.Rotation.X {type=MovieSceneScriptingFloatChannel, default=0, const [0,50]}
       leg_l_pv_ik_ctrl.Rotation.Y {type=MovieSceneScriptingFloatChannel, default=0, const [0,50]}
       leg_l_pv_ik_ctrl.Rotation.Z {type=MovieSceneScriptingFloatChannel, default=0, const [0,50]}
       leg_l_pv_ik_ctrl.Scale.X {type=MovieSceneScriptingFloatChannel, default=1, const [1,50]}
       leg_l_pv_ik_ctrl.Scale.Y {type=MovieSceneScriptingFloatChannel, default=1, const [1,50]}
       leg_l_pv_ik_ctrl.Scale.Z {type=MovieSceneScriptingFloatChannel, default=1, const [1,50]}
       thigh_r_fk_ctrl.Location.X {type=MovieSceneScriptingFloatChannel, default=0, thinned_from=50, keys [[0,0],[328,0]]}
       thigh_r_fk_ctrl.Location.Y {type=MovieSceneScriptingFloatChannel, default=0, thinned_from=50, keys [[0,0],[328,0]]}
       thigh_r_fk_ctrl.Location.Z {type=MovieSceneScriptingFloatChannel, default=0, thinned_from=50, keys [[0,0],[328,0]]}
       thigh_r_fk_ctrl.Rotation.X {
        type=MovieSceneScriptingFloatChannel
        default=1.0528
        thinned_from=50
        keys [[0,15.4578],[14,15.4578],[26,15.3253],[35,15.4391],[118,15.4578],[126,15.7373],[134,15.8124],[140,15.8201],[145,15.7222],[151,15.5591],[159,15.4613],[249,15.4163],[328,15.4578]]
       }
       thigh_r_fk_ctrl.Rotation.Y {
        type=MovieSceneScriptingFloatChannel
        default=-1.026
        thinned_from=50
        keys [[0,-1.0638],[14,-1.0638],[26,-1.0777],[35,-1.0658],[118,-1.0638],[122,-0.8752],[126,-0.6757],[130,-0.6038],[134,-0.5715],[140,-0.5607],[145,-0.862],[151,-1.3642],[159,-1.6655],[184,-1.8209],[192,-1.8463],[238,-1.8463],[244,-1.7729],[249,-1.6507],[254,-1.2839],[260,-1.1463],[266,-1.0948],[273,-1.0754],[281,-1.0638],[328,-1.0638]]
       }
       thigh_r_fk_ctrl.Rotation.Z {
        type=MovieSceneScriptingFloatChannel
        default=-0.07
        thinned_from=50
        keys [[0,-2.2712],[14,-2.2712],[26,-2.4092],[32,-2.3229],[35,-2.2906],[39,-2.2782],[53,-2.2712],[118,-2.2712],[126,-2.2029],[130,-2.1903],[134,-2.1846],[238,-2.1843],[244,-2.1925],[249,-2.2061],[254,-2.2467],[260,-2.262],[281,-2.2712],[328,-2.2712]]
       }
       thigh_r_fk_ctrl.Scale.X {type=MovieSceneScriptingFloatChannel, default=1, const [1,50]}
       thigh_r_fk_ctrl.Scale.Y {type=MovieSceneScriptingFloatChannel, default=1, const [1,50]}
       thigh_r_fk_ctrl.Scale.Z {type=MovieSceneScriptingFloatChannel, default=1, const [1,50]}
       calf_r_fk_ctrl.Location.X {
        type=MovieSceneScriptingFloatChannel
        default=0
        thinned_from=50
        keys [[0,-0.6147],[14,-0.6147],[26,-0.6738],[32,-0.6369],[35,-0.623],[39,-0.6178],[53,-0.6147],[118,-0.6147],[126,-0.6003],[130,-0.5976],[140,-0.596],[145,-0.6109],[151,-0.6357],[159,-0.6506],[184,-0.6582],[238,-0.6595],[244,-0.6553],[249,-0.6483],[254,-0.6273],[260,-0.6195],[266,-0.6165],[281,-0.6147],[328,-0.6147]]
       }
       calf_r_fk_ctrl.Location.Y {
        type=MovieSceneScriptingFloatChannel
        default=0
        thinned_from=50
        keys [[0,0.0821],[14,0.0821],[26,0.09],[32,0.0851],[35,0.0832],[39,0.0825],[53,0.0821],[118,0.0821],[126,0.0802],[134,0.0796],[140,0.0796],[145,0.0816],[151,0.0849],[159,0.0869],[184,0.0879],[238,0.0881],[244,0.0875],[249,0.0866],[254,0.0838],[260,0.0827],[266,0.0823],[328,0.0821]]
       }
       calf_r_fk_ctrl.Location.Z {
        type=MovieSceneScriptingFloatChannel
        default=0
        thinned_from=50
        keys [[0,-0.0112],[14,-0.0112],[26,-0.0123],[35,-0.0114],[118,-0.0112],[140,-0.0109],[151,-0.0116],[159,-0.0119],[244,-0.012],[249,-0.0118],[254,-0.0115],[328,-0.0112]]
       }
       calf_r_fk_ctrl.Rotation.X {type=MovieSceneScriptingFloatChannel, default=0.0752, thinned_from=50, keys [[0,-0.0638],[328,-0.0638]]}
       calf_r_fk_ctrl.Rotation.Y {type=MovieSceneScriptingFloatChannel, default=1.0359, thinned_from=50, keys [[0,1.0367],[328,1.0367]]}
       calf_r_fk_ctrl.Rotation.Z {type=MovieSceneScriptingFloatChannel, default=0.0678, thinned_from=50, keys [[0,-7.6066],[328,-7.6066]]}
       calf_r_fk_ctrl.Scale.X {type=MovieSceneScriptingFloatChannel, default=1, const [1,50]}
       calf_r_fk_ctrl.Scale.Y {type=MovieSceneScriptingFloatChannel, default=1, const [1,50]}
       calf_r_fk_ctrl.Scale.Z {type=MovieSceneScriptingFloatChannel, default=1, const [1,50]}
       foot_r_fk_ctrl.Location.X {
        type=MovieSceneScriptingFloatChannel
        default=0
        thinned_from=50
        keys [[0,-0.6825],[14,-0.6825],[26,-0.7482],[32,-0.7071],[35,-0.6918],[39,-0.6859],[53,-0.6825],[118,-0.6825],[126,-0.6665],[130,-0.6635],[140,-0.6618],[145,-0.6783],[151,-0.7058],[159,-0.7223],[184,-0.7308],[238,-0.7322],[244,-0.7276],[249,-0.7198],[254,-0.6965],[260,-0.6878],[266,-0.6845],[281,-0.6825],[328,-0.6825]]
       }
       foot_r_fk_ctrl.Location.Y {type=MovieSceneScriptingFloatChannel, default=-0.0543, thinned_from=50, keys [[0,-0.0543],[328,-0.0543]]}
       foot_r_fk_ctrl.Location.Z {type=MovieSceneScriptingFloatChannel, default=-0.8609, thinned_from=50, keys [[0,-0.8609],[328,-0.8609]]}
       foot_r_fk_ctrl.Rotation.X {
        type=MovieSceneScriptingFloatChannel
        default=-0.9802
        thinned_from=50
        keys [[0,-1.1508],[26,-1.1488],[140,-1.1481],[145,-1.1387],[151,-1.1229],[159,-1.1134],[238,-1.1077],[244,-1.1118],[249,-1.1185],[254,-1.1387],[260,-1.1462],[281,-1.1508],[328,-1.1508]]
       }
       foot_r_fk_ctrl.Rotation.Y {
        type=MovieSceneScriptingFloatChannel
        default=0.1214
        thinned_from=50
        keys [[0,1.3021],[14,1.3021],[26,1.3214],[35,1.3048],[118,1.3021],[134,1.2958],[140,1.2957],[145,1.3528],[151,1.4481],[159,1.5053],[184,1.5348],[192,1.5396],[238,1.5396],[244,1.5173],[249,1.4802],[254,1.3689],[260,1.3272],[266,1.3115],[273,1.3056],[281,1.3021],[328,1.3021]]
       }
       foot_r_fk_ctrl.Rotation.Z {
        type=MovieSceneScriptingFloatChannel
        default=0.0013
        thinned_from=50
        keys [[0,8.5714],[14,8.5714],[26,8.5891],[118,8.5714],[126,8.5177],[140,8.5018],[151,8.4462],[159,8.4253],[238,8.4128],[249,8.4525],[254,8.5268],[260,8.5546],[328,8.5714]]
       }
       foot_r_fk_ctrl.Scale.X {type=MovieSceneScriptingFloatChannel, default=1, const [1,50]}
       foot_r_fk_ctrl.Scale.Y {type=MovieSceneScriptingFloatChannel, default=1, const [1,50]}
       foot_r_fk_ctrl.Scale.Z {type=MovieSceneScriptingFloatChannel, default=1, const [1,50]}
       ball_r_fk_ctrl.Location.X {type=MovieSceneScriptingFloatChannel, default=0, const [0,50]}
       ball_r_fk_ctrl.Location.Y {type=MovieSceneScriptingFloatChannel, default=0, const [0,50]}
       ball_r_fk_ctrl.Location.Z {type=MovieSceneScriptingFloatChannel, default=0, const [0,50]}
       ball_r_fk_ctrl.Rotation.X {type=MovieSceneScriptingFloatChannel, default=0, const [0,50]}
       ball_r_fk_ctrl.Rotation.Y {type=MovieSceneScriptingFloatChannel, default=0, const [0,50]}
       ball_r_fk_ctrl.Rotation.Z {type=MovieSceneScriptingFloatChannel, default=0, const [0,50]}
       ball_r_fk_ctrl.Scale.X {type=MovieSceneScriptingFloatChannel, default=1, const [1,50]}
       ball_r_fk_ctrl.Scale.Y {type=MovieSceneScriptingFloatChannel, default=1, const [1,50]}
       ball_r_fk_ctrl.Scale.Z {type=MovieSceneScriptingFloatChannel, default=1, const [1,50]}
       foot_r_ik_ctrl.Location.X {type=MovieSceneScriptingFloatChannel, default=0, const [-2.4852,50]}
       foot_r_ik_ctrl.Location.Y {type=MovieSceneScriptingFloatChannel, default=0, const [-11.0272,50]}
       foot_r_ik_ctrl.Location.Z {type=MovieSceneScriptingFloatChannel, default=0, const [0,50]}
       foot_r_ik_ctrl.Rotation.X {type=MovieSceneScriptingFloatChannel, default=0, const [0,50]}
       foot_r_ik_ctrl.Rotation.Y {type=MovieSceneScriptingFloatChannel, default=0, const [0,50]}
       foot_r_ik_ctrl.Rotation.Z {type=MovieSceneScriptingFloatChannel, default=0, const [-38,50]}
       foot_r_ik_ctrl.Scale.X {type=MovieSceneScriptingFloatChannel, default=1, const [1,50]}
       foot_r_ik_ctrl.Scale.Y {type=MovieSceneScriptingFloatChannel, default=1, const [1,50]}
       foot_r_ik_ctrl.Scale.Z {type=MovieSceneScriptingFloatChannel, default=1, const [1,50]}
       ball_r_ik_ctrl.Location.X {type=MovieSceneScriptingFloatChannel, default=0, const [0,50]}
       ball_r_ik_ctrl.Location.Y {type=MovieSceneScriptingFloatChannel, default=0, const [0,50]}
       ball_r_ik_ctrl.Location.Z {type=MovieSceneScriptingFloatChannel, default=0, const [0,50]}
       ball_r_ik_ctrl.Rotation.X {type=MovieSceneScriptingFloatChannel, default=0, const [0,50]}
       ball_r_ik_ctrl.Rotation.Y {type=MovieSceneScriptingFloatChannel, default=0, const [0,50]}
       ball_r_ik_ctrl.Rotation.Z {type=MovieSceneScriptingFloatChannel, default=0, const [0,50]}
       ball_r_ik_ctrl.Scale.X {type=MovieSceneScriptingFloatChannel, default=1, const [1,50]}
       ball_r_ik_ctrl.Scale.Y {type=MovieSceneScriptingFloatChannel, default=1, const [1,50]}
       ball_r_ik_ctrl.Scale.Z {type=MovieSceneScriptingFloatChannel, default=1, const [1,50]}
       foot_bk1_r_ctrl.Location.X {type=MovieSceneScriptingFloatChannel, default=0, const [0,50]}
       foot_bk1_r_ctrl.Location.Y {type=MovieSceneScriptingFloatChannel, default=0, const [0,50]}
       foot_bk1_r_ctrl.Location.Z {type=MovieSceneScriptingFloatChannel, default=0, const [0,50]}
       foot_bk1_r_ctrl.Rotation.X {type=MovieSceneScriptingFloatChannel, default=0, const [0,50]}
       foot_bk1_r_ctrl.Rotation.Y {type=MovieSceneScriptingFloatChannel, default=0, const [0,50]}
       foot_bk1_r_ctrl.Rotation.Z {type=MovieSceneScriptingFloatChannel, default=0, const [0,50]}
       foot_bk1_r_ctrl.Scale.X {type=MovieSceneScriptingFloatChannel, default=1, const [1,50]}
       foot_bk1_r_ctrl.Scale.Y {type=MovieSceneScriptingFloatChannel, default=1, const [1,50]}
       foot_bk1_r_ctrl.Scale.Z {type=MovieSceneScriptingFloatChannel, default=1, const [1,50]}
       foot_roll_r_ctrl.Location.X {type=MovieSceneScriptingFloatChannel, default=0, const [0,50]}
       foot_roll_r_ctrl.Location.Y {type=MovieSceneScriptingFloatChannel, default=0, const [0,50]}
       foot_roll_r_ctrl.Location.Z {type=MovieSceneScriptingFloatChannel, default=0, const [0,50]}
       foot_roll_r_ctrl.Rotation.X {type=MovieSceneScriptingFloatChannel, default=0, const [0,50]}
       foot_roll_r_ctrl.Rotation.Y {type=MovieSceneScriptingFloatChannel, default=0, const [0,50]}
       foot_roll_r_ctrl.Rotation.Z {type=MovieSceneScriptingFloatChannel, default=0, const [0,50]}
       foot_roll_r_ctrl.Scale.X {type=MovieSceneScriptingFloatChannel, default=1, const [1,50]}
       foot_roll_r_ctrl.Scale.Y {type=MovieSceneScriptingFloatChannel, default=1, const [1,50]}
       foot_roll_r_ctrl.Scale.Z {type=MovieSceneScriptingFloatChannel, default=1, const [1,50]}
       roll_blend_r_ctrl {type=MovieSceneScriptingFloatChannel, default=1, const [1,50]}
       tip_r_ctrl.Location.X {type=MovieSceneScriptingFloatChannel, default=0, const [0,50]}
       tip_r_ctrl.Location.Y {type=MovieSceneScriptingFloatChannel, default=0, const [0,50]}
       tip_r_ctrl.Location.Z {type=MovieSceneScriptingFloatChannel, default=0, const [0,50]}
       tip_r_ctrl.Rotation.X {type=MovieSceneScriptingFloatChannel, default=0, const [0,50]}
       tip_r_ctrl.Rotation.Y {type=MovieSceneScriptingFloatChannel, default=0, const [0,50]}
       tip_r_ctrl.Rotation.Z {type=MovieSceneScriptingFloatChannel, default=0, const [0,50]}
       tip_r_ctrl.Scale.X {type=MovieSceneScriptingFloatChannel, default=1, const [1,50]}
       tip_r_ctrl.Scale.Y {type=MovieSceneScriptingFloatChannel, default=1, const [1,50]}
       tip_r_ctrl.Scale.Z {type=MovieSceneScriptingFloatChannel, default=1, const [1,50]}
       heel_r_ctrl.Location.X {type=MovieSceneScriptingFloatChannel, default=0, const [0,50]}
       heel_r_ctrl.Location.Y {type=MovieSceneScriptingFloatChannel, default=0, const [0,50]}
       heel_r_ctrl.Location.Z {type=MovieSceneScriptingFloatChannel, default=0, const [0,50]}
       heel_r_ctrl.Rotation.X {type=MovieSceneScriptingFloatChannel, default=0, const [0,50]}
       heel_r_ctrl.Rotation.Y {type=MovieSceneScriptingFloatChannel, default=0, const [0,50]}
       heel_r_ctrl.Rotation.Z {type=MovieSceneScriptingFloatChannel, default=0, const [0,50]}
       heel_r_ctrl.Scale.X {type=MovieSceneScriptingFloatChannel, default=1, const [1,50]}
       heel_r_ctrl.Scale.Y {type=MovieSceneScriptingFloatChannel, default=1, const [1,50]}
       heel_r_ctrl.Scale.Z {type=MovieSceneScriptingFloatChannel, default=1, const [1,50]}
       leg_r_softness {type=MovieSceneScriptingFloatChannel, default=0, const [0,50]}
       leg_r_pv_ik_ctrl.Location.X {type=MovieSceneScriptingFloatChannel, default=0, const [0,50]}
       leg_r_pv_ik_ctrl.Location.Y {type=MovieSceneScriptingFloatChannel, default=0, const [0,50]}
       leg_r_pv_ik_ctrl.Location.Z {type=MovieSceneScriptingFloatChannel, default=0, const [0,50]}
       leg_r_pv_ik_ctrl.Rotation.X {type=MovieSceneScriptingFloatChannel, default=0, const [0,50]}
       leg_r_pv_ik_ctrl.Rotation.Y {type=MovieSceneScriptingFloatChannel, default=0, const [0,50]}
       leg_r_pv_ik_ctrl.Rotation.Z {type=MovieSceneScriptingFloatChannel, default=0, const [0,50]}
       leg_r_pv_ik_ctrl.Scale.X {type=MovieSceneScriptingFloatChannel, default=1, const [1,50]}
       leg_r_pv_ik_ctrl.Scale.Y {type=MovieSceneScriptingFloatChannel, default=1, const [1,50]}
       leg_r_pv_ik_ctrl.Scale.Z {type=MovieSceneScriptingFloatChannel, default=1, const [1,50]}
       upperarm_l_fk_ctrl.Location.X {type=MovieSceneScriptingFloatChannel, default=0, const [0,52]}
       upperarm_l_fk_ctrl.Location.Y {type=MovieSceneScriptingFloatChannel, default=0, const [0,52]}
       upperarm_l_fk_ctrl.Location.Z {type=MovieSceneScriptingFloatChannel, default=0, const [0,52]}
       upperarm_l_fk_ctrl.Rotation.X {type=MovieSceneScriptingFloatChannel, default=0, const [24.7416,51]}
       upperarm_l_fk_ctrl.Rotation.Y {
        type=MovieSceneScriptingFloatChannel
        default=0
        thinned_from=51
        keys [[0,-8.8794],[26,-8.8794],[32,-8.7287],[35,-8.5881],[39,-8.5336],[45,-8.641],[53,-8.8403],[60,-8.9323],[87,-8.9691],[104,-8.9523],[108,-8.9197],[111,-8.8656],[114,-8.8334],[118,-8.8253],[134,-8.8922],[328,-8.8794]]
       }
       upperarm_l_fk_ctrl.Rotation.Z {
        type=MovieSceneScriptingFloatChannel
        default=0
        thinned_from=28
        keys [[0,-22.0291],[7,-22.133],[14,-22.1551],[26,-21.5832],[35,-20.4842],[39,-20.0567],[46,-20.8936],[65,-23.3398],[92,-22.8772],[102,-23.129],[109,-23.3708],[122,-22.7423],[131,-21.9489],[159,-21.1266],[190,-21.3168],[201,-20.9268],[207,-21.2358],[224,-22.647],[237,-23.111],[254,-23.265],[260,-23.1762],[287,-22.2478],[297,-21.9655],[304,-21.8797],[328,-22.0291]]
       }
       upperarm_l_fk_ctrl.Scale.X {type=MovieSceneScriptingFloatChannel, default=1, const [1,51]}
       upperarm_l_fk_ctrl.Scale.Y {type=MovieSceneScriptingFloatChannel, default=1, const [1,51]}
       upperarm_l_fk_ctrl.Scale.Z {type=MovieSceneScriptingFloatChannel, default=1, const [1,51]}
       lowerarm_l_fk_ctrl.Location.X {type=MovieSceneScriptingFloatChannel, default=0, const [0,50]}
       lowerarm_l_fk_ctrl.Location.Y {type=MovieSceneScriptingFloatChannel, default=0, const [0,50]}
       lowerarm_l_fk_ctrl.Location.Z {type=MovieSceneScriptingFloatChannel, default=0, const [0,50]}
       lowerarm_l_fk_ctrl.Rotation.X {type=MovieSceneScriptingFloatChannel, default=0, const [0,50]}
       lowerarm_l_fk_ctrl.Rotation.Y {type=MovieSceneScriptingFloatChannel, default=0, const [0,50]}
       lowerarm_l_fk_ctrl.Rotation.Z {
        type=MovieSceneScriptingFloatChannel
        default=0
        thinned_from=33
        keys [[0,9.302],[7,9.3392],[14,9.4383],[32,9.9385],[35,10.0342],[39,10.1129],[47,10.018],[56,9.6095],[66,8.6765],[71,8.2494],[82,8.5893],[92,8.6933],[108,8.1172],[125,8.8317],[139,9.2029],[153,9.2875],[160,9.238],[178,8.669],[185,8.3708],[195,8.1731],[210,8.7968],[227,8.4897],[247,8.0485],[258,7.9291],[266,7.8869],[278,8.0041],[286,8.2756],[293,8.7479],[300,9.0316],[308,9.1843],[316,9.2664],[328,9.302]]
       }
       lowerarm_l_fk_ctrl.Scale.X {type=MovieSceneScriptingFloatChannel, default=1, const [1,50]}
       lowerarm_l_fk_ctrl.Scale.Y {type=MovieSceneScriptingFloatChannel, default=1, const [1,50]}
       lowerarm_l_fk_ctrl.Scale.Z {type=MovieSceneScriptingFloatChannel, default=1, const [1,50]}
       hand_l_fk_ctrl.Location.X {type=MovieSceneScriptingFloatChannel, default=0, const [0,50]}
       hand_l_fk_ctrl.Location.Y {type=MovieSceneScriptingFloatChannel, default=0, const [0,50]}
       hand_l_fk_ctrl.Location.Z {type=MovieSceneScriptingFloatChannel, default=0, const [0,50]}
       hand_l_fk_ctrl.Rotation.X {type=MovieSceneScriptingFloatChannel, default=0, const [7,50]}
       hand_l_fk_ctrl.Rotation.Y {type=MovieSceneScriptingFloatChannel, default=0, const [6.5,50]}
       hand_l_fk_ctrl.Rotation.Z {type=MovieSceneScriptingFloatChannel, default=0, const [-6.5,50]}
       hand_l_fk_ctrl.Scale.X {type=MovieSceneScriptingFloatChannel, default=1, const [1,50]}
       hand_l_fk_ctrl.Scale.Y {type=MovieSceneScriptingFloatChannel, default=1, const [1,50]}
       hand_l_fk_ctrl.Scale.Z {type=MovieSceneScriptingFloatChannel, default=1, const [1,50]}
       thumb_01_l_ctrl.Location.X {type=MovieSceneScriptingFloatChannel, default=0, const [0,50]}
       thumb_01_l_ctrl.Location.Y {type=MovieSceneScriptingFloatChannel, default=0, const [0,50]}
       thumb_01_l_ctrl.Location.Z {type=MovieSceneScriptingFloatChannel, default=0, const [0,50]}
       thumb_01_l_ctrl.Rotation.X {type=MovieSceneScriptingFloatChannel, default=0, const [-17,50]}
       thumb_01_l_ctrl.Rotation.Y {type=MovieSceneScriptingFloatChannel, default=0, const [3,50]}
       thumb_01_l_ctrl.Rotation.Z {type=MovieSceneScriptingFloatChannel, default=0, const [-6,50]}
       thumb_01_l_ctrl.Scale.X {type=MovieSceneScriptingFloatChannel, default=1, const [1,50]}
       thumb_01_l_ctrl.Scale.Y {type=MovieSceneScriptingFloatChannel, default=1, const [1,50]}
       thumb_01_l_ctrl.Scale.Z {type=MovieSceneScriptingFloatChannel, default=1, const [1,50]}
       thumb_02_l_ctrl.Location.X {type=MovieSceneScriptingFloatChannel, default=0, const [0,50]}
       thumb_02_l_ctrl.Location.Y {type=MovieSceneScriptingFloatChannel, default=0, const [0,50]}
       thumb_02_l_ctrl.Location.Z {type=MovieSceneScriptingFloatChannel, default=0, const [0,50]}
       thumb_02_l_ctrl.Rotation.X {type=MovieSceneScriptingFloatChannel, default=0, const [0,50]}
       thumb_02_l_ctrl.Rotation.Y {type=MovieSceneScriptingFloatChannel, default=0, const [0,50]}
       thumb_02_l_ctrl.Rotation.Z {type=MovieSceneScriptingFloatChannel, default=0, const [-15.5,50]}
       thumb_02_l_ctrl.Scale.X {type=MovieSceneScriptingFloatChannel, default=1, const [1,50]}
       thumb_02_l_ctrl.Scale.Y {type=MovieSceneScriptingFloatChannel, default=1, const [1,50]}
       thumb_02_l_ctrl.Scale.Z {type=MovieSceneScriptingFloatChannel, default=1, const [1,50]}
       thumb_03_l_ctrl.Location.X {type=MovieSceneScriptingFloatChannel, default=0, const [0,50]}
       thumb_03_l_ctrl.Location.Y {type=MovieSceneScriptingFloatChannel, default=0, const [0,50]}
       thumb_03_l_ctrl.Location.Z {type=MovieSceneScriptingFloatChannel, default=0, const [0,50]}
       thumb_03_l_ctrl.Rotation.X {type=MovieSceneScriptingFloatChannel, default=0, const [0,50]}
       thumb_03_l_ctrl.Rotation.Y {type=MovieSceneScriptingFloatChannel, default=0, const [0,50]}
       thumb_03_l_ctrl.Rotation.Z {type=MovieSceneScriptingFloatChannel, default=0, const [-44,50]}
       thumb_03_l_ctrl.Scale.X {type=MovieSceneScriptingFloatChannel, default=1, const [1,50]}
       thumb_03_l_ctrl.Scale.Y {type=MovieSceneScriptingFloatChannel, default=1, const [1,50]}
       thumb_03_l_ctrl.Scale.Z {type=MovieSceneScriptingFloatChannel, default=1, const [1,50]}
       index_metacarpal_l_ctrl.Location.X {type=MovieSceneScriptingFloatChannel, default=0, const [0,50]}
       index_metacarpal_l_ctrl.Location.Y {type=MovieSceneScriptingFloatChannel, default=0, const [0,50]}
       index_metacarpal_l_ctrl.Location.Z {type=MovieSceneScriptingFloatChannel, default=0, const [0,50]}
       index_metacarpal_l_ctrl.Rotation.X {type=MovieSceneScriptingFloatChannel, default=0, const [0,50]}
       index_metacarpal_l_ctrl.Rotation.Y {type=MovieSceneScriptingFloatChannel, default=0, const [0,50]}
       index_metacarpal_l_ctrl.Rotation.Z {type=MovieSceneScriptingFloatChannel, default=0, const [0,50]}
       index_metacarpal_l_ctrl.Scale.X {type=MovieSceneScriptingFloatChannel, default=1, const [1,50]}
       index_metacarpal_l_ctrl.Scale.Y {type=MovieSceneScriptingFloatChannel, default=1, const [1,50]}
       index_metacarpal_l_ctrl.Scale.Z {type=MovieSceneScriptingFloatChannel, default=1, const [1,50]}
       index_01_l_ctrl.Location.X {type=MovieSceneScriptingFloatChannel, default=0, const [0,50]}
       index_01_l_ctrl.Location.Y {type=MovieSceneScriptingFloatChannel, default=0, const [0,50]}
       index_01_l_ctrl.Location.Z {type=MovieSceneScriptingFloatChannel, default=0, const [0,50]}
       index_01_l_ctrl.Rotation.X {type=MovieSceneScriptingFloatChannel, default=0, const [0,50]}
       index_01_l_ctrl.Rotation.Y {type=MovieSceneScriptingFloatChannel, default=0, const [7.5,50]}
       index_01_l_ctrl.Rotation.Z {type=MovieSceneScriptingFloatChannel, default=0, const [-46.5,50]}
       index_01_l_ctrl.Scale.X {type=MovieSceneScriptingFloatChannel, default=1, const [1.1358,50]}
       index_01_l_ctrl.Scale.Y {type=MovieSceneScriptingFloatChannel, default=1, const [1,50]}
       index_01_l_ctrl.Scale.Z {type=MovieSceneScriptingFloatChannel, default=1, const [1,50]}
       index_02_l_ctrl.Location.X {type=MovieSceneScriptingFloatChannel, default=0, const [0,50]}
       index_02_l_ctrl.Location.Y {type=MovieSceneScriptingFloatChannel, default=0, const [0,50]}
       index_02_l_ctrl.Location.Z {type=MovieSceneScriptingFloatChannel, default=0, const [0,50]}
       index_02_l_ctrl.Rotation.X {type=MovieSceneScriptingFloatChannel, default=0, const [0,50]}
       index_02_l_ctrl.Rotation.Y {type=MovieSceneScriptingFloatChannel, default=0, const [0,50]}
       index_02_l_ctrl.Rotation.Z {type=MovieSceneScriptingFloatChannel, default=0, const [-51.5,50]}
       index_02_l_ctrl.Scale.X {type=MovieSceneScriptingFloatChannel, default=1, const [1,50]}
       index_02_l_ctrl.Scale.Y {type=MovieSceneScriptingFloatChannel, default=1, const [1,50]}
       index_02_l_ctrl.Scale.Z {type=MovieSceneScriptingFloatChannel, default=1, const [1,50]}
       index_03_l_ctrl.Location.X {type=MovieSceneScriptingFloatChannel, default=0, const [0,50]}
       index_03_l_ctrl.Location.Y {type=MovieSceneScriptingFloatChannel, default=0, const [0,50]}
       index_03_l_ctrl.Location.Z {type=MovieSceneScriptingFloatChannel, default=0, const [0,50]}
       index_03_l_ctrl.Rotation.X {type=MovieSceneScriptingFloatChannel, default=0, const [0,50]}
       index_03_l_ctrl.Rotation.Y {type=MovieSceneScriptingFloatChannel, default=0, const [0,50]}
       index_03_l_ctrl.Rotation.Z {type=MovieSceneScriptingFloatChannel, default=0, const [-45.5,50]}
       index_03_l_ctrl.Scale.X {type=MovieSceneScriptingFloatChannel, default=1, const [1,50]}
       index_03_l_ctrl.Scale.Y {type=MovieSceneScriptingFloatChannel, default=1, const [1,50]}
       index_03_l_ctrl.Scale.Z {type=MovieSceneScriptingFloatChannel, default=1, const [1,50]}
       middle_metacarpal_l_ctrl.Location.X {type=MovieSceneScriptingFloatChannel, default=0, const [0,50]}
       middle_metacarpal_l_ctrl.Location.Y {type=MovieSceneScriptingFloatChannel, default=0, const [0,50]}
       middle_metacarpal_l_ctrl.Location.Z {type=MovieSceneScriptingFloatChannel, default=0, const [0,50]}
       middle_metacarpal_l_ctrl.Rotation.X {type=MovieSceneScriptingFloatChannel, default=0, const [0,50]}
       middle_metacarpal_l_ctrl.Rotation.Y {type=MovieSceneScriptingFloatChannel, default=0, const [-1.5,50]}
       middle_metacarpal_l_ctrl.Rotation.Z {type=MovieSceneScriptingFloatChannel, default=0, const [1,50]}
       middle_metacarpal_l_ctrl.Scale.X {type=MovieSceneScriptingFloatChannel, default=1, const [1,50]}
       middle_metacarpal_l_ctrl.Scale.Y {type=MovieSceneScriptingFloatChannel, default=1, const [1,50]}
       middle_metacarpal_l_ctrl.Scale.Z {type=MovieSceneScriptingFloatChannel, default=1, const [1,50]}
       middle_01_l_ctrl.Location.X {type=MovieSceneScriptingFloatChannel, default=0, const [0,50]}
       middle_01_l_ctrl.Location.Y {type=MovieSceneScriptingFloatChannel, default=0, const [0,50]}
       middle_01_l_ctrl.Location.Z {type=MovieSceneScriptingFloatChannel, default=0, const [0,50]}
       middle_01_l_ctrl.Rotation.X {type=MovieSceneScriptingFloatChannel, default=0, const [-6,50]}
       middle_01_l_ctrl.Rotation.Y {type=MovieSceneScriptingFloatChannel, default=0, const [0,50]}
       middle_01_l_ctrl.Rotation.Z {type=MovieSceneScriptingFloatChannel, default=0, const [-56.5,50]}
       middle_01_l_ctrl.Scale.X {type=MovieSceneScriptingFloatChannel, default=1, const [1,50]}
       middle_01_l_ctrl.Scale.Y {type=MovieSceneScriptingFloatChannel, default=1, const [1,50]}
       middle_01_l_ctrl.Scale.Z {type=MovieSceneScriptingFloatChannel, default=1, const [1,50]}
       middle_02_l_ctrl.Location.X {type=MovieSceneScriptingFloatChannel, default=0, const [0,50]}
       middle_02_l_ctrl.Location.Y {type=MovieSceneScriptingFloatChannel, default=0, const [0,50]}
       middle_02_l_ctrl.Location.Z {type=MovieSceneScriptingFloatChannel, default=0, const [0,50]}
       middle_02_l_ctrl.Rotation.X {type=MovieSceneScriptingFloatChannel, default=0, const [0,50]}
       middle_02_l_ctrl.Rotation.Y {type=MovieSceneScriptingFloatChannel, default=0, const [0,50]}
       middle_02_l_ctrl.Rotation.Z {type=MovieSceneScriptingFloatChannel, default=0, const [-52.5,50]}
       middle_02_l_ctrl.Scale.X {type=MovieSceneScriptingFloatChannel, default=1, const [1,50]}
       middle_02_l_ctrl.Scale.Y {type=MovieSceneScriptingFloatChannel, default=1, const [1,50]}
       middle_02_l_ctrl.Scale.Z {type=MovieSceneScriptingFloatChannel, default=1, const [1,50]}
       middle_03_l_ctrl.Location.X {type=MovieSceneScriptingFloatChannel, default=0, const [0,50]}
       middle_03_l_ctrl.Location.Y {type=MovieSceneScriptingFloatChannel, default=0, const [0,50]}
       middle_03_l_ctrl.Location.Z {type=MovieSceneScriptingFloatChannel, default=0, const [0,50]}
       middle_03_l_ctrl.Rotation.X {type=MovieSceneScriptingFloatChannel, default=0, const [0,50]}
       middle_03_l_ctrl.Rotation.Y {type=MovieSceneScriptingFloatChannel, default=0, const [0,50]}
       middle_03_l_ctrl.Rotation.Z {type=MovieSceneScriptingFloatChannel, default=0, const [-27.5,50]}
       middle_03_l_ctrl.Scale.X {type=MovieSceneScriptingFloatChannel, default=1, const [1,50]}
       middle_03_l_ctrl.Scale.Y {type=MovieSceneScriptingFloatChannel, default=1, const [1,50]}
       middle_03_l_ctrl.Scale.Z {type=MovieSceneScriptingFloatChannel, default=1, const [1,50]}
       ring_metacarpal_l_ctrl.Location.X {type=MovieSceneScriptingFloatChannel, default=0, const [0,50]}
       ring_metacarpal_l_ctrl.Location.Y {type=MovieSceneScriptingFloatChannel, default=0, const [0,50]}
       ring_metacarpal_l_ctrl.Location.Z {type=MovieSceneScriptingFloatChannel, default=0, const [0,50]}
       ring_metacarpal_l_ctrl.Rotation.X {type=MovieSceneScriptingFloatChannel, default=0, const [0,50]}
       ring_metacarpal_l_ctrl.Rotation.Y {type=MovieSceneScriptingFloatChannel, default=0, const [-5,50]}
       ring_metacarpal_l_ctrl.Rotation.Z {type=MovieSceneScriptingFloatChannel, default=0, const [1,50]}
       ring_metacarpal_l_ctrl.Scale.X {type=MovieSceneScriptingFloatChannel, default=1, const [1,50]}
       ring_metacarpal_l_ctrl.Scale.Y {type=MovieSceneScriptingFloatChannel, default=1, const [1,50]}
       ring_metacarpal_l_ctrl.Scale.Z {type=MovieSceneScriptingFloatChannel, default=1, const [1,50]}
       ring_01_l_ctrl.Location.X {type=MovieSceneScriptingFloatChannel, default=0, const [0,50]}
       ring_01_l_ctrl.Location.Y {type=MovieSceneScriptingFloatChannel, default=0, const [0,50]}
       ring_01_l_ctrl.Location.Z {type=MovieSceneScriptingFloatChannel, default=0, const [0,50]}
       ring_01_l_ctrl.Rotation.X {type=MovieSceneScriptingFloatChannel, default=0, const [-12.3417,50]}
       ring_01_l_ctrl.Rotation.Y {type=MovieSceneScriptingFloatChannel, default=0, const [5.3087,50]}
       ring_01_l_ctrl.Rotation.Z {type=MovieSceneScriptingFloatChannel, default=0, const [-53.3181,50]}
       ring_01_l_ctrl.Scale.X {type=MovieSceneScriptingFloatChannel, default=1, const [1,50]}
       ring_01_l_ctrl.Scale.Y {type=MovieSceneScriptingFloatChannel, default=1, const [1,50]}
       ring_01_l_ctrl.Scale.Z {type=MovieSceneScriptingFloatChannel, default=1, const [1,50]}
       ring_02_l_ctrl.Location.X {type=MovieSceneScriptingFloatChannel, default=0, const [0,50]}
       ring_02_l_ctrl.Location.Y {type=MovieSceneScriptingFloatChannel, default=0, const [0,50]}
       ring_02_l_ctrl.Location.Z {type=MovieSceneScriptingFloatChannel, default=0, const [0,50]}
       ring_02_l_ctrl.Rotation.X {type=MovieSceneScriptingFloatChannel, default=0, const [0,50]}
       ring_02_l_ctrl.Rotation.Y {type=MovieSceneScriptingFloatChannel, default=0, const [0,50]}
       ring_02_l_ctrl.Rotation.Z {type=MovieSceneScriptingFloatChannel, default=0, const [-55,50]}
       ring_02_l_ctrl.Scale.X {type=MovieSceneScriptingFloatChannel, default=1, const [1,50]}
       ring_02_l_ctrl.Scale.Y {type=MovieSceneScriptingFloatChannel, default=1, const [1,50]}
       ring_02_l_ctrl.Scale.Z {type=MovieSceneScriptingFloatChannel, default=1, const [1,50]}
       ring_03_l_ctrl.Location.X {type=MovieSceneScriptingFloatChannel, default=0, const [0,50]}
       ring_03_l_ctrl.Location.Y {type=MovieSceneScriptingFloatChannel, default=0, const [0,50]}
       ring_03_l_ctrl.Location.Z {type=MovieSceneScriptingFloatChannel, default=0, const [0,50]}
       ring_03_l_ctrl.Rotation.X {type=MovieSceneScriptingFloatChannel, default=0, const [0,50]}
       ring_03_l_ctrl.Rotation.Y {type=MovieSceneScriptingFloatChannel, default=0, const [0,50]}
       ring_03_l_ctrl.Rotation.Z {type=MovieSceneScriptingFloatChannel, default=0, const [-37.5,50]}
       ring_03_l_ctrl.Scale.X {type=MovieSceneScriptingFloatChannel, default=1, const [1,50]}
       ring_03_l_ctrl.Scale.Y {type=MovieSceneScriptingFloatChannel, default=1, const [1,50]}
       ring_03_l_ctrl.Scale.Z {type=MovieSceneScriptingFloatChannel, default=1, const [1,50]}
       pinky_metacarpal_l_ctrl.Location.X {type=MovieSceneScriptingFloatChannel, default=0, const [0,50]}
       pinky_metacarpal_l_ctrl.Location.Y {type=MovieSceneScriptingFloatChannel, default=0, const [0,50]}
       pinky_metacarpal_l_ctrl.Location.Z {type=MovieSceneScriptingFloatChannel, default=0, const [0,50]}
       pinky_metacarpal_l_ctrl.Rotation.X {type=MovieSceneScriptingFloatChannel, default=0, const [0,50]}
       pinky_metacarpal_l_ctrl.Rotation.Y {type=MovieSceneScriptingFloatChannel, default=0, const [-5,50]}
       pinky_metacarpal_l_ctrl.Rotation.Z {type=MovieSceneScriptingFloatChannel, default=0, const [0,50]}
       pinky_metacarpal_l_ctrl.Scale.X {type=MovieSceneScriptingFloatChannel, default=1, const [1,50]}
       pinky_metacarpal_l_ctrl.Scale.Y {type=MovieSceneScriptingFloatChannel, default=1, const [1,50]}
       pinky_metacarpal_l_ctrl.Scale.Z {type=MovieSceneScriptingFloatChannel, default=1, const [1,50]}
       pinky_01_l_ctrl.Location.X {type=MovieSceneScriptingFloatChannel, default=0, const [0,50]}
       pinky_01_l_ctrl.Location.Y {type=MovieSceneScriptingFloatChannel, default=0, const [0,50]}
       pinky_01_l_ctrl.Location.Z {type=MovieSceneScriptingFloatChannel, default=0, const [0,50]}
       pinky_01_l_ctrl.Rotation.X {type=MovieSceneScriptingFloatChannel, default=0, const [-17.9299,50]}
       pinky_01_l_ctrl.Rotation.Y {type=MovieSceneScriptingFloatChannel, default=0, const [3.0717,50]}
       pinky_01_l_ctrl.Rotation.Z {type=MovieSceneScriptingFloatChannel, default=0, const [-38.5,50]}
       pinky_01_l_ctrl.Scale.X {type=MovieSceneScriptingFloatChannel, default=1, const [1,50]}
       pinky_01_l_ctrl.Scale.Y {type=MovieSceneScriptingFloatChannel, default=1, const [1,50]}
       pinky_01_l_ctrl.Scale.Z {type=MovieSceneScriptingFloatChannel, default=1, const [1,50]}
       pinky_02_l_ctrl.Location.X {type=MovieSceneScriptingFloatChannel, default=0, const [0,50]}
       pinky_02_l_ctrl.Location.Y {type=MovieSceneScriptingFloatChannel, default=0, const [0,50]}
       pinky_02_l_ctrl.Location.Z {type=MovieSceneScriptingFloatChannel, default=0, const [0,50]}
       pinky_02_l_ctrl.Rotation.X {type=MovieSceneScriptingFloatChannel, default=0, const [0,50]}
       pinky_02_l_ctrl.Rotation.Y {type=MovieSceneScriptingFloatChannel, default=0, const [0,50]}
       pinky_02_l_ctrl.Rotation.Z {type=MovieSceneScriptingFloatChannel, default=0, const [-48.5,50]}
       pinky_02_l_ctrl.Scale.X {type=MovieSceneScriptingFloatChannel, default=1, const [1,50]}
       pinky_02_l_ctrl.Scale.Y {type=MovieSceneScriptingFloatChannel, default=1, const [1,50]}
       pinky_02_l_ctrl.Scale.Z {type=MovieSceneScriptingFloatChannel, default=1, const [1,50]}
       pinky_03_l_ctrl.Location.X {type=MovieSceneScriptingFloatChannel, default=0, const [0,50]}
       pinky_03_l_ctrl.Location.Y {type=MovieSceneScriptingFloatChannel, default=0, const [0,50]}
       pinky_03_l_ctrl.Location.Z {type=MovieSceneScriptingFloatChannel, default=0, const [0,50]}
       pinky_03_l_ctrl.Rotation.X {type=MovieSceneScriptingFloatChannel, default=0, const [0,50]}
       pinky_03_l_ctrl.Rotation.Y {type=MovieSceneScriptingFloatChannel, default=0, const [0,50]}
       pinky_03_l_ctrl.Rotation.Z {type=MovieSceneScriptingFloatChannel, default=0, const [-43.5,50]}
       pinky_03_l_ctrl.Scale.X {type=MovieSceneScriptingFloatChannel, default=1, const [1,50]}
       pinky_03_l_ctrl.Scale.Y {type=MovieSceneScriptingFloatChannel, default=1, const [1,50]}
       pinky_03_l_ctrl.Scale.Z {type=MovieSceneScriptingFloatChannel, default=1, const [1,50]}
       hand_l_ik_ctrl.Location.X {
        type=MovieSceneScriptingFloatChannel
        default=0
        thinned_from=50
        keys [[0,-2.6544],[14,-2.6544],[26,-2.7947],[32,-2.707],[35,-2.6741],[39,-2.6615],[60,-2.6544],[118,-2.6544],[122,-2.8904],[126,-3.14],[130,-3.2298],[134,-3.2702],[140,-3.2837],[145,-3.2157],[151,-3.1022],[159,-3.0342],[184,-2.9991],[238,-2.9933],[244,-2.9615],[249,-2.9086],[254,-2.7497],[260,-2.6901],[266,-2.6678],[281,-2.6544],[328,-2.6544]]
       }
       hand_l_ik_ctrl.Location.Y {
        type=MovieSceneScriptingFloatChannel
        default=0
        thinned_from=50
        keys [[0,5.5173],[14,5.5173],[26,5.4694],[35,5.5106],[118,5.5173],[126,5.1694],[130,5.105],[134,5.076],[140,5.0663],[145,5.2441],[151,5.5403],[159,5.718],[184,5.8096],[238,5.8246],[244,5.7958],[249,5.7478],[254,5.6037],[260,5.5497],[266,5.5295],[328,5.5173]]
       }
       hand_l_ik_ctrl.Location.Z {
        type=MovieSceneScriptingFloatChannel
        default=0
        thinned_from=50
        keys [[0,-1.3623],[14,-1.3623],[26,-0.8638],[32,-1.1753],[35,-1.2922],[39,-1.3368],[45,-1.3527],[53,-1.3623],[118,-1.3623],[122,-1.599],[126,-1.8494],[130,-1.9395],[134,-1.9801],[140,-1.9936],[145,-1.7293],[151,-1.2888],[159,-1.0245],[177,-0.9254],[184,-0.8882],[192,-0.8659],[238,-0.8659],[244,-0.9124],[249,-0.99],[254,-1.2227],[260,-1.3099],[266,-1.3426],[273,-1.3549],[281,-1.3623],[328,-1.3623]]
       }
       hand_l_ik_ctrl.Rotation.X {
        type=MovieSceneScriptingFloatChannel
        default=0
        thinned_from=50
        keys [[0,29.4727],[14,29.4727],[26,29.6317],[35,29.4951],[118,29.4727],[126,28.4116],[130,28.2152],[134,28.1269],[140,28.0974],[145,28.5478],[151,29.2983],[159,29.7486],[192,30.0188],[238,30.0188],[249,29.8823],[254,29.6263],[260,29.5303],[328,29.4727]]
       }
       hand_l_ik_ctrl.Rotation.Y {
        type=MovieSceneScriptingFloatChannel
        default=0
        thinned_from=50
        keys [[0,10.5534],[14,10.5534],[26,11.0111],[32,10.7251],[35,10.6178],[39,10.5768],[118,10.5534],[126,10.2881],[130,10.239],[134,10.217],[140,10.2096],[145,10.308],[151,10.4721],[159,10.5705],[238,10.6296],[249,10.6105],[254,10.5748],[328,10.5534]]
       }
       hand_l_ik_ctrl.Rotation.Z {
        type=MovieSceneScriptingFloatChannel
        default=0
        thinned_from=50
        keys [[0,14.6764],[14,14.6764],[26,14.8696],[35,14.7035],[118,14.6764],[126,13.6656],[130,13.4786],[134,13.3944],[140,13.3663],[145,13.8673],[151,14.7023],[159,15.2032],[184,15.4615],[192,15.5038],[238,15.5038],[244,15.4262],[249,15.2969],[254,14.9091],[260,14.7636],[266,14.7091],[328,14.6764]]
       }
       hand_l_ik_ctrl.Scale.X {type=MovieSceneScriptingFloatChannel, default=1, const [1,50]}
       hand_l_ik_ctrl.Scale.Y {type=MovieSceneScriptingFloatChannel, default=1, const [1,50]}
       hand_l_ik_ctrl.Scale.Z {type=MovieSceneScriptingFloatChannel, default=1, const [1,50]}
       Softness {type=MovieSceneScriptingFloatChannel, default=0, const [0,50]}
       hand_l_ikorient_ctrl.Rotation.Roll {type=MovieSceneScriptingFloatChannel, default=0, const [0,50]}
       hand_l_ikorient_ctrl.Rotation.Pitch {type=MovieSceneScriptingFloatChannel, default=0, const [0,50]}
       hand_l_ikorient_ctrl.Rotation.Yaw {type=MovieSceneScriptingFloatChannel, default=0, const [0,50]}
       arm_l_pv_ik_ctrl.Location.X {
        type=MovieSceneScriptingFloatChannel
        default=0
        thinned_from=50
        keys [[0,9.3783],[14,9.3783],[26,9.4586],[35,9.3896],[118,9.3783],[126,8.8126],[130,8.7079],[134,8.6608],[140,8.6451],[145,8.8168],[151,9.1031],[159,9.2748],[184,9.3634],[328,9.3783]]
       }
       arm_l_pv_ik_ctrl.Location.Y {
        type=MovieSceneScriptingFloatChannel
        default=0
        thinned_from=50
        keys [[0,-0.4488],[14,-0.4488],[26,-0.1031],[32,-0.3192],[35,-0.4002],[39,-0.4312],[45,-0.4422],[53,-0.4488],[118,-0.4488],[122,-0.8904],[126,-1.3573],[130,-1.5254],[134,-1.601],[140,-1.6262],[145,-1.2353],[151,-0.5836],[159,-0.1926],[177,-0.046],[184,0.009],[192,0.042],[238,0.042],[244,-0.004],[249,-0.0807],[254,-0.3108],[260,-0.3971],[266,-0.4294],[273,-0.4415],[281,-0.4488],[328,-0.4488]]
       }
       arm_l_pv_ik_ctrl.Location.Z {
        type=MovieSceneScriptingFloatChannel
        default=0
        thinned_from=50
        keys [[0,5.3334],[14,5.3334],[26,5.4406],[32,5.3736],[35,5.3485],[60,5.3334],[118,5.3334],[126,5.6897],[130,5.7557],[134,5.7854],[140,5.7953],[145,5.7624],[151,5.7075],[159,5.6746],[238,5.6549],[244,5.6247],[249,5.5745],[254,5.4238],[260,5.3673],[266,5.3461],[328,5.3334]]
       }
       arm_l_pv_ik_ctrl.Rotation.X {
        type=MovieSceneScriptingFloatChannel
        default=0
        thinned_from=50
        keys [[0,12.4045],[14,12.4045],[26,12.0026],[32,12.2538],[35,12.348],[39,12.384],[145,12.4143],[159,12.3893],[328,12.4045]]
       }
       arm_l_pv_ik_ctrl.Rotation.Y {
        type=MovieSceneScriptingFloatChannel
        default=0
        thinned_from=50
        keys [[0,-7.3861],[14,-7.3861],[26,-7.3541],[39,-7.3845],[118,-7.3861],[126,-7.041],[130,-6.9771],[134,-6.9484],[140,-6.9388],[145,-7.1536],[151,-7.5116],[159,-7.7265],[184,-7.8372],[238,-7.8554],[244,-7.8114],[249,-7.738],[254,-7.5181],[260,-7.4356],[266,-7.4046],[328,-7.3861]]
       }
       arm_l_pv_ik_ctrl.Rotation.Z {
        type=MovieSceneScriptingFloatChannel
        default=0
        thinned_from=50
        keys [[0,23.4217],[14,23.4217],[26,23.7906],[35,23.4735],[118,23.4217],[126,22.1113],[130,21.8689],[134,21.7597],[140,21.7234],[145,22.3039],[151,23.2714],[159,23.8519],[184,24.1513],[238,24.2003],[244,24.1273],[249,24.0056],[254,23.6406],[260,23.5038],[328,23.4217]]
       }
       arm_l_pv_ik_ctrl.Scale.X {type=MovieSceneScriptingFloatChannel, default=1, const [1,50]}
       arm_l_pv_ik_ctrl.Scale.Y {type=MovieSceneScriptingFloatChannel, default=1, const [1,50]}
       arm_l_pv_ik_ctrl.Scale.Z {type=MovieSceneScriptingFloatChannel, default=1, const [1,50]}
       upperarm_r_fk_ctrl.Location.X {type=MovieSceneScriptingFloatChannel, default=0, const [0,50]}
       upperarm_r_fk_ctrl.Location.Y {type=MovieSceneScriptingFloatChannel, default=0, const [0,50]}
       upperarm_r_fk_ctrl.Location.Z {type=MovieSceneScriptingFloatChannel, default=0, const [0,50]}
       upperarm_r_fk_ctrl.Rotation.X {type=MovieSceneScriptingFloatChannel, default=0, const [3.9444,50]}
       upperarm_r_fk_ctrl.Rotation.Y {
        type=MovieSceneScriptingFloatChannel
        default=0
        thinned_from=29
        keys [[0,-8.5222],[14,-8.5222],[26,-8.6018],[39,-9.2665],[45,-9.3063],[53,-9.2701],[87,-9.0384],[102,-8.9686],[108,-8.9575],[134,-9.4216],[140,-9.4732],[210,-9.4732],[289,-8.5222],[328,-8.5222]]
       }
       upperarm_r_fk_ctrl.Rotation.Z {
        type=MovieSceneScriptingFloatChannel
        default=0
        thinned_from=29
        keys [[0,-18.9877],[14,-18.8212],[24,-18.7417],[31,-18.7981],[35,-18.9507],[39,-19.2176],[50,-18.7275],[58,-18.4787],[64,-18.4133],[73,-18.6524],[88,-19.249],[100,-18.9969],[110,-18.2998],[114,-17.9515],[123,-18.6576],[135,-19.2533],[177,-17.9656],[190,-17.8263],[211,-18.9992],[239,-18.1845],[266,-17.7758],[273,-17.872],[310,-18.7792],[318,-18.9281],[328,-18.9877]]
       }
       upperarm_r_fk_ctrl.Scale.X {type=MovieSceneScriptingFloatChannel, default=1, const [1,50]}
       upperarm_r_fk_ctrl.Scale.Y {type=MovieSceneScriptingFloatChannel, default=1, const [1,50]}
       upperarm_r_fk_ctrl.Scale.Z {type=MovieSceneScriptingFloatChannel, default=1, const [1,50]}
       lowerarm_r_fk_ctrl.Location.X {type=MovieSceneScriptingFloatChannel, default=0, const [0,52]}
       lowerarm_r_fk_ctrl.Location.Y {type=MovieSceneScriptingFloatChannel, default=0, const [0,52]}
       lowerarm_r_fk_ctrl.Location.Z {type=MovieSceneScriptingFloatChannel, default=0, const [0,52]}
       lowerarm_r_fk_ctrl.Rotation.X {type=MovieSceneScriptingFloatChannel, default=0, const [0,52]}
       lowerarm_r_fk_ctrl.Rotation.Y {type=MovieSceneScriptingFloatChannel, default=0, const [0,52]}
       lowerarm_r_fk_ctrl.Rotation.Z {
        type=MovieSceneScriptingFloatChannel
        default=0
        thinned_from=33
        keys [[0,4.258],[7,4.0435],[14,3.8085],[23,3.6724],[29,3.8652],[36,4.4649],[44,5.8236],[54,6.9222],[76,7.2769],[95,7.3343],[110,7.4776],[115,7.3902],[124,6.7396],[133,6.1567],[145,5.8942],[160,6.7842],[179,8.4814],[197,9.0779],[217,7.1481],[230,7.9369],[238,8.4053],[244,8.5946],[263,8.7323],[267,8.6846],[273,8.3519],[281,7.5219],[289,6.3186],[296,5.171],[303,4.3791],[310,4.0964],[318,4.1772],[328,4.258]]
       }
       lowerarm_r_fk_ctrl.Scale.X {type=MovieSceneScriptingFloatChannel, default=1, const [1,52]}
       lowerarm_r_fk_ctrl.Scale.Y {type=MovieSceneScriptingFloatChannel, default=1, const [1,52]}
       lowerarm_r_fk_ctrl.Scale.Z {type=MovieSceneScriptingFloatChannel, default=1, const [1,52]}
       hand_r_fk_ctrl.Location.X {type=MovieSceneScriptingFloatChannel, default=0, const [0,50]}
       hand_r_fk_ctrl.Location.Y {type=MovieSceneScriptingFloatChannel, default=0, const [0,50]}
       hand_r_fk_ctrl.Location.Z {type=MovieSceneScriptingFloatChannel, default=0, const [0,50]}
       hand_r_fk_ctrl.Rotation.X {type=MovieSceneScriptingFloatChannel, default=0, const [-19.5,50]}
       hand_r_fk_ctrl.Rotation.Y {type=MovieSceneScriptingFloatChannel, default=0, const [10.5,50]}
       hand_r_fk_ctrl.Rotation.Z {type=MovieSceneScriptingFloatChannel, default=0, const [-4.5,50]}
       hand_r_fk_ctrl.Scale.X {type=MovieSceneScriptingFloatChannel, default=1, const [1,50]}
       hand_r_fk_ctrl.Scale.Y {type=MovieSceneScriptingFloatChannel, default=1, const [1,50]}
       hand_r_fk_ctrl.Scale.Z {type=MovieSceneScriptingFloatChannel, default=1, const [1,50]}
       thumb_01_r_ctrl.Location.X {type=MovieSceneScriptingFloatChannel, default=0, const [0,50]}
       thumb_01_r_ctrl.Location.Y {type=MovieSceneScriptingFloatChannel, default=0, const [0,50]}
       thumb_01_r_ctrl.Location.Z {type=MovieSceneScriptingFloatChannel, default=0, const [0,50]}
       thumb_01_r_ctrl.Rotation.X {type=MovieSceneScriptingFloatChannel, default=0, const [-8,50]}
       thumb_01_r_ctrl.Rotation.Y {type=MovieSceneScriptingFloatChannel, default=0, const [3,50]}
       thumb_01_r_ctrl.Rotation.Z {type=MovieSceneScriptingFloatChannel, default=0, const [-0.5,50]}
       thumb_01_r_ctrl.Scale.X {type=MovieSceneScriptingFloatChannel, default=1, const [1,50]}
       thumb_01_r_ctrl.Scale.Y {type=MovieSceneScriptingFloatChannel, default=1, const [1,50]}
       thumb_01_r_ctrl.Scale.Z {type=MovieSceneScriptingFloatChannel, default=1, const [1,50]}
       thumb_02_r_ctrl.Location.X {type=MovieSceneScriptingFloatChannel, default=0, const [0,50]}
       thumb_02_r_ctrl.Location.Y {type=MovieSceneScriptingFloatChannel, default=0, const [0,50]}
       thumb_02_r_ctrl.Location.Z {type=MovieSceneScriptingFloatChannel, default=0, const [0,50]}
       thumb_02_r_ctrl.Rotation.X {type=MovieSceneScriptingFloatChannel, default=0, const [0,50]}
       thumb_02_r_ctrl.Rotation.Y {type=MovieSceneScriptingFloatChannel, default=0, const [0,50]}
       thumb_02_r_ctrl.Rotation.Z {type=MovieSceneScriptingFloatChannel, default=0, const [-15.5,50]}
       thumb_02_r_ctrl.Scale.X {type=MovieSceneScriptingFloatChannel, default=1, const [1,50]}
       thumb_02_r_ctrl.Scale.Y {type=MovieSceneScriptingFloatChannel, default=1, const [1,50]}
       thumb_02_r_ctrl.Scale.Z {type=MovieSceneScriptingFloatChannel, default=1, const [1,50]}
       thumb_03_r_ctrl.Location.X {type=MovieSceneScriptingFloatChannel, default=0, const [0,50]}
       thumb_03_r_ctrl.Location.Y {type=MovieSceneScriptingFloatChannel, default=0, const [0,50]}
       thumb_03_r_ctrl.Location.Z {type=MovieSceneScriptingFloatChannel, default=0, const [0,50]}
       thumb_03_r_ctrl.Rotation.X {type=MovieSceneScriptingFloatChannel, default=0, const [0,50]}
       thumb_03_r_ctrl.Rotation.Y {type=MovieSceneScriptingFloatChannel, default=0, const [0,50]}
       thumb_03_r_ctrl.Rotation.Z {type=MovieSceneScriptingFloatChannel, default=0, const [-49,50]}
       thumb_03_r_ctrl.Scale.X {type=MovieSceneScriptingFloatChannel, default=1, const [1,50]}
       thumb_03_r_ctrl.Scale.Y {type=MovieSceneScriptingFloatChannel, default=1, const [1,50]}
       thumb_03_r_ctrl.Scale.Z {type=MovieSceneScriptingFloatChannel, default=1, const [1,50]}
       index_metacarpal_r_ctrl.Location.X {type=MovieSceneScriptingFloatChannel, default=0, const [0,50]}
       index_metacarpal_r_ctrl.Location.Y {type=MovieSceneScriptingFloatChannel, default=0, const [0,50]}
       index_metacarpal_r_ctrl.Location.Z {type=MovieSceneScriptingFloatChannel, default=0, const [0,50]}
       index_metacarpal_r_ctrl.Rotation.X {type=MovieSceneScriptingFloatChannel, default=0, const [0,50]}
       index_metacarpal_r_ctrl.Rotation.Y {type=MovieSceneScriptingFloatChannel, default=0, const [1,50]}
       index_metacarpal_r_ctrl.Rotation.Z {type=MovieSceneScriptingFloatChannel, default=0, const [0,50]}
       index_metacarpal_r_ctrl.Scale.X {type=MovieSceneScriptingFloatChannel, default=1, const [1,50]}
       index_metacarpal_r_ctrl.Scale.Y {type=MovieSceneScriptingFloatChannel, default=1, const [1,50]}
       index_metacarpal_r_ctrl.Scale.Z {type=MovieSceneScriptingFloatChannel, default=1, const [1,50]}
       index_01_r_ctrl.Location.X {type=MovieSceneScriptingFloatChannel, default=0, const [0.0561,50]}
       index_01_r_ctrl.Location.Y {type=MovieSceneScriptingFloatChannel, default=0, const [-0.0068,50]}
       index_01_r_ctrl.Location.Z {type=MovieSceneScriptingFloatChannel, default=0, const [0.029,50]}
       index_01_r_ctrl.Rotation.X {type=MovieSceneScriptingFloatChannel, default=0, const [-3.5,50]}
       index_01_r_ctrl.Rotation.Y {type=MovieSceneScriptingFloatChannel, default=0, const [5,50]}
       index_01_r_ctrl.Rotation.Z {type=MovieSceneScriptingFloatChannel, default=0, const [-47,50]}
       index_01_r_ctrl.Scale.X {type=MovieSceneScriptingFloatChannel, default=1, const [1.0561,50]}
       index_01_r_ctrl.Scale.Y {type=MovieSceneScriptingFloatChannel, default=1, const [1,50]}
       index_01_r_ctrl.Scale.Z {type=MovieSceneScriptingFloatChannel, default=1, const [1,50]}
       index_02_r_ctrl.Location.X {type=MovieSceneScriptingFloatChannel, default=0, const [0,50]}
       index_02_r_ctrl.Location.Y {type=MovieSceneScriptingFloatChannel, default=0, const [0,50]}
       index_02_r_ctrl.Location.Z {type=MovieSceneScriptingFloatChannel, default=0, const [0,50]}
       index_02_r_ctrl.Rotation.X {type=MovieSceneScriptingFloatChannel, default=0, const [0,50]}
       index_02_r_ctrl.Rotation.Y {type=MovieSceneScriptingFloatChannel, default=0, const [0,50]}
       index_02_r_ctrl.Rotation.Z {type=MovieSceneScriptingFloatChannel, default=0, const [-51.5,50]}
       index_02_r_ctrl.Scale.X {type=MovieSceneScriptingFloatChannel, default=1, const [1,50]}
       index_02_r_ctrl.Scale.Y {type=MovieSceneScriptingFloatChannel, default=1, const [1,50]}
       index_02_r_ctrl.Scale.Z {type=MovieSceneScriptingFloatChannel, default=1, const [1,50]}
       index_03_r_ctrl.Location.X {type=MovieSceneScriptingFloatChannel, default=0, const [0,50]}
       index_03_r_ctrl.Location.Y {type=MovieSceneScriptingFloatChannel, default=0, const [0,50]}
       index_03_r_ctrl.Location.Z {type=MovieSceneScriptingFloatChannel, default=0, const [0,50]}
       index_03_r_ctrl.Rotation.X {type=MovieSceneScriptingFloatChannel, default=0, const [0,50]}
       index_03_r_ctrl.Rotation.Y {type=MovieSceneScriptingFloatChannel, default=0, const [0,50]}
       index_03_r_ctrl.Rotation.Z {type=MovieSceneScriptingFloatChannel, default=0, const [-37,50]}
       index_03_r_ctrl.Scale.X {type=MovieSceneScriptingFloatChannel, default=1, const [1,50]}
       index_03_r_ctrl.Scale.Y {type=MovieSceneScriptingFloatChannel, default=1, const [1,50]}
       index_03_r_ctrl.Scale.Z {type=MovieSceneScriptingFloatChannel, default=1, const [1,50]}
       middle_metacarpal_r_ctrl.Location.X {type=MovieSceneScriptingFloatChannel, default=0, const [0,50]}
       middle_metacarpal_r_ctrl.Location.Y {type=MovieSceneScriptingFloatChannel, default=0, const [0,50]}
       middle_metacarpal_r_ctrl.Location.Z {type=MovieSceneScriptingFloatChannel, default=0, const [0,50]}
       middle_metacarpal_r_ctrl.Rotation.X {type=MovieSceneScriptingFloatChannel, default=0, const [0,50]}
       middle_metacarpal_r_ctrl.Rotation.Y {type=MovieSceneScriptingFloatChannel, default=0, const [-1.5,50]}
       middle_metacarpal_r_ctrl.Rotation.Z {type=MovieSceneScriptingFloatChannel, default=0, const [1,50]}
       middle_metacarpal_r_ctrl.Scale.X {type=MovieSceneScriptingFloatChannel, default=1, const [1,50]}
       middle_metacarpal_r_ctrl.Scale.Y {type=MovieSceneScriptingFloatChannel, default=1, const [1,50]}
       middle_metacarpal_r_ctrl.Scale.Z {type=MovieSceneScriptingFloatChannel, default=1, const [1,50]}
       middle_01_r_ctrl.Location.X {type=MovieSceneScriptingFloatChannel, default=0, const [0,50]}
       middle_01_r_ctrl.Location.Y {type=MovieSceneScriptingFloatChannel, default=0, const [0,50]}
       middle_01_r_ctrl.Location.Z {type=MovieSceneScriptingFloatChannel, default=0, const [0,50]}
       middle_01_r_ctrl.Rotation.X {type=MovieSceneScriptingFloatChannel, default=0, const [-6,50]}
       middle_01_r_ctrl.Rotation.Y {type=MovieSceneScriptingFloatChannel, default=0, const [0,50]}
       middle_01_r_ctrl.Rotation.Z {type=MovieSceneScriptingFloatChannel, default=0, const [-56.5,50]}
       middle_01_r_ctrl.Scale.X {type=MovieSceneScriptingFloatChannel, default=1, const [1,50]}
       middle_01_r_ctrl.Scale.Y {type=MovieSceneScriptingFloatChannel, default=1, const [1,50]}
       middle_01_r_ctrl.Scale.Z {type=MovieSceneScriptingFloatChannel, default=1, const [1,50]}
       middle_02_r_ctrl.Location.X {type=MovieSceneScriptingFloatChannel, default=0, const [0,50]}
       middle_02_r_ctrl.Location.Y {type=MovieSceneScriptingFloatChannel, default=0, const [0,50]}
       middle_02_r_ctrl.Location.Z {type=MovieSceneScriptingFloatChannel, default=0, const [0,50]}
       middle_02_r_ctrl.Rotation.X {type=MovieSceneScriptingFloatChannel, default=0, const [0,50]}
       middle_02_r_ctrl.Rotation.Y {type=MovieSceneScriptingFloatChannel, default=0, const [0,50]}
       middle_02_r_ctrl.Rotation.Z {type=MovieSceneScriptingFloatChannel, default=0, const [-53,50]}
       middle_02_r_ctrl.Scale.X {type=MovieSceneScriptingFloatChannel, default=1, const [1,50]}
       middle_02_r_ctrl.Scale.Y {type=MovieSceneScriptingFloatChannel, default=1, const [1,50]}
       middle_02_r_ctrl.Scale.Z {type=MovieSceneScriptingFloatChannel, default=1, const [1,50]}
       middle_03_r_ctrl.Location.X {type=MovieSceneScriptingFloatChannel, default=0, const [0,50]}
       middle_03_r_ctrl.Location.Y {type=MovieSceneScriptingFloatChannel, default=0, const [0,50]}
       middle_03_r_ctrl.Location.Z {type=MovieSceneScriptingFloatChannel, default=0, const [0,50]}
       middle_03_r_ctrl.Rotation.X {type=MovieSceneScriptingFloatChannel, default=0, const [0,50]}
       middle_03_r_ctrl.Rotation.Y {type=MovieSceneScriptingFloatChannel, default=0, const [0,50]}
       middle_03_r_ctrl.Rotation.Z {type=MovieSceneScriptingFloatChannel, default=0, const [-29.5,50]}
       middle_03_r_ctrl.Scale.X {type=MovieSceneScriptingFloatChannel, default=1, const [1,50]}
       middle_03_r_ctrl.Scale.Y {type=MovieSceneScriptingFloatChannel, default=1, const [1,50]}
       middle_03_r_ctrl.Scale.Z {type=MovieSceneScriptingFloatChannel, default=1, const [1,50]}
       ring_metacarpal_r_ctrl.Location.X {type=MovieSceneScriptingFloatChannel, default=0, const [0,50]}
       ring_metacarpal_r_ctrl.Location.Y {type=MovieSceneScriptingFloatChannel, default=0, const [0,50]}
       ring_metacarpal_r_ctrl.Location.Z {type=MovieSceneScriptingFloatChannel, default=0, const [0,50]}
       ring_metacarpal_r_ctrl.Rotation.X {type=MovieSceneScriptingFloatChannel, default=0, const [0,50]}
       ring_metacarpal_r_ctrl.Rotation.Y {type=MovieSceneScriptingFloatChannel, default=0, const [-5,50]}
       ring_metacarpal_r_ctrl.Rotation.Z {type=MovieSceneScriptingFloatChannel, default=0, const [1,50]}
       ring_metacarpal_r_ctrl.Scale.X {type=MovieSceneScriptingFloatChannel, default=1, const [1,50]}
       ring_metacarpal_r_ctrl.Scale.Y {type=MovieSceneScriptingFloatChannel, default=1, const [1,50]}
       ring_metacarpal_r_ctrl.Scale.Z {type=MovieSceneScriptingFloatChannel, default=1, const [1,50]}
       ring_01_r_ctrl.Location.X {type=MovieSceneScriptingFloatChannel, default=0, const [0,50]}
       ring_01_r_ctrl.Location.Y {type=MovieSceneScriptingFloatChannel, default=0, const [0,50]}
       ring_01_r_ctrl.Location.Z {type=MovieSceneScriptingFloatChannel, default=0, const [0,50]}
       ring_01_r_ctrl.Rotation.X {type=MovieSceneScriptingFloatChannel, default=0, const [-12.3417,50]}
       ring_01_r_ctrl.Rotation.Y {type=MovieSceneScriptingFloatChannel, default=0, const [5.3087,50]}
       ring_01_r_ctrl.Rotation.Z {type=MovieSceneScriptingFloatChannel, default=0, const [-53.3181,50]}
       ring_01_r_ctrl.Scale.X {type=MovieSceneScriptingFloatChannel, default=1, const [1,50]}
       ring_01_r_ctrl.Scale.Y {type=MovieSceneScriptingFloatChannel, default=1, const [1,50]}
       ring_01_r_ctrl.Scale.Z {type=MovieSceneScriptingFloatChannel, default=1, const [1,50]}
       ring_02_r_ctrl.Location.X {type=MovieSceneScriptingFloatChannel, default=0, const [0,50]}
       ring_02_r_ctrl.Location.Y {type=MovieSceneScriptingFloatChannel, default=0, const [0,50]}
       ring_02_r_ctrl.Location.Z {type=MovieSceneScriptingFloatChannel, default=0, const [0,50]}
       ring_02_r_ctrl.Rotation.X {type=MovieSceneScriptingFloatChannel, default=0, const [0,50]}
       ring_02_r_ctrl.Rotation.Y {type=MovieSceneScriptingFloatChannel, default=0, const [0,50]}
       ring_02_r_ctrl.Rotation.Z {type=MovieSceneScriptingFloatChannel, default=0, const [-56,50]}
       ring_02_r_ctrl.Scale.X {type=MovieSceneScriptingFloatChannel, default=1, const [1,50]}
       ring_02_r_ctrl.Scale.Y {type=MovieSceneScriptingFloatChannel, default=1, const [1,50]}
       ring_02_r_ctrl.Scale.Z {type=MovieSceneScriptingFloatChannel, default=1, const [1,50]}
       ring_03_r_ctrl.Location.X {type=MovieSceneScriptingFloatChannel, default=0, const [0,50]}
       ring_03_r_ctrl.Location.Y {type=MovieSceneScriptingFloatChannel, default=0, const [0,50]}
       ring_03_r_ctrl.Location.Z {type=MovieSceneScriptingFloatChannel, default=0, const [0,50]}
       ring_03_r_ctrl.Rotation.X {type=MovieSceneScriptingFloatChannel, default=0, const [0,50]}
       ring_03_r_ctrl.Rotation.Y {type=MovieSceneScriptingFloatChannel, default=0, const [0,50]}
       ring_03_r_ctrl.Rotation.Z {type=MovieSceneScriptingFloatChannel, default=0, const [-37.5,50]}
       ring_03_r_ctrl.Scale.X {type=MovieSceneScriptingFloatChannel, default=1, const [1,50]}
       ring_03_r_ctrl.Scale.Y {type=MovieSceneScriptingFloatChannel, default=1, const [1,50]}
       ring_03_r_ctrl.Scale.Z {type=MovieSceneScriptingFloatChannel, default=1, const [1,50]}
       pinky_metacarpal_r_ctrl.Location.X {type=MovieSceneScriptingFloatChannel, default=0, const [0,50]}
       pinky_metacarpal_r_ctrl.Location.Y {type=MovieSceneScriptingFloatChannel, default=0, const [0,50]}
       pinky_metacarpal_r_ctrl.Location.Z {type=MovieSceneScriptingFloatChannel, default=0, const [0,50]}
       pinky_metacarpal_r_ctrl.Rotation.X {type=MovieSceneScriptingFloatChannel, default=0, const [0,50]}
       pinky_metacarpal_r_ctrl.Rotation.Y {type=MovieSceneScriptingFloatChannel, default=0, const [-5,50]}
       pinky_metacarpal_r_ctrl.Rotation.Z {type=MovieSceneScriptingFloatChannel, default=0, const [0,50]}
       pinky_metacarpal_r_ctrl.Scale.X {type=MovieSceneScriptingFloatChannel, default=1, const [1,50]}
       pinky_metacarpal_r_ctrl.Scale.Y {type=MovieSceneScriptingFloatChannel, default=1, const [1,50]}
       pinky_metacarpal_r_ctrl.Scale.Z {type=MovieSceneScriptingFloatChannel, default=1, const [1,50]}
       pinky_01_r_ctrl.Location.X {type=MovieSceneScriptingFloatChannel, default=0, const [0,50]}
       pinky_01_r_ctrl.Location.Y {type=MovieSceneScriptingFloatChannel, default=0, const [0,50]}
       pinky_01_r_ctrl.Location.Z {type=MovieSceneScriptingFloatChannel, default=0, const [0,50]}
       pinky_01_r_ctrl.Rotation.X {type=MovieSceneScriptingFloatChannel, default=0, const [-17.9299,50]}
       pinky_01_r_ctrl.Rotation.Y {type=MovieSceneScriptingFloatChannel, default=0, const [3.0717,50]}
       pinky_01_r_ctrl.Rotation.Z {type=MovieSceneScriptingFloatChannel, default=0, const [-47.5,50]}
       pinky_01_r_ctrl.Scale.X {type=MovieSceneScriptingFloatChannel, default=1, const [1,50]}
       pinky_01_r_ctrl.Scale.Y {type=MovieSceneScriptingFloatChannel, default=1, const [1,50]}
       pinky_01_r_ctrl.Scale.Z {type=MovieSceneScriptingFloatChannel, default=1, const [1,50]}
       pinky_02_r_ctrl.Location.X {type=MovieSceneScriptingFloatChannel, default=0, const [0,50]}
       pinky_02_r_ctrl.Location.Y {type=MovieSceneScriptingFloatChannel, default=0, const [0,50]}
       pinky_02_r_ctrl.Location.Z {type=MovieSceneScriptingFloatChannel, default=0, const [0,50]}
       pinky_02_r_ctrl.Rotation.X {type=MovieSceneScriptingFloatChannel, default=0, const [0,50]}
       pinky_02_r_ctrl.Rotation.Y {type=MovieSceneScriptingFloatChannel, default=0, const [0,50]}
       pinky_02_r_ctrl.Rotation.Z {type=MovieSceneScriptingFloatChannel, default=0, const [-48.5,50]}
       pinky_02_r_ctrl.Scale.X {type=MovieSceneScriptingFloatChannel, default=1, const [1,50]}
       pinky_02_r_ctrl.Scale.Y {type=MovieSceneScriptingFloatChannel, default=1, const [1,50]}
       pinky_02_r_ctrl.Scale.Z {type=MovieSceneScriptingFloatChannel, default=1, const [1,50]}
       pinky_03_r_ctrl.Location.X {type=MovieSceneScriptingFloatChannel, default=0, const [0,50]}
       pinky_03_r_ctrl.Location.Y {type=MovieSceneScriptingFloatChannel, default=0, const [0,50]}
       pinky_03_r_ctrl.Location.Z {type=MovieSceneScriptingFloatChannel, default=0, const [0,50]}
       pinky_03_r_ctrl.Rotation.X {type=MovieSceneScriptingFloatChannel, default=0, const [0,50]}
       pinky_03_r_ctrl.Rotation.Y {type=MovieSceneScriptingFloatChannel, default=0, const [0,50]}
       pinky_03_r_ctrl.Rotation.Z {type=MovieSceneScriptingFloatChannel, default=0, const [-43.5,50]}
       pinky_03_r_ctrl.Scale.X {type=MovieSceneScriptingFloatChannel, default=1, const [1,50]}
       pinky_03_r_ctrl.Scale.Y {type=MovieSceneScriptingFloatChannel, default=1, const [1,50]}
       pinky_03_r_ctrl.Scale.Z {type=MovieSceneScriptingFloatChannel, default=1, const [1,50]}
       hand_r_ik_ctrl.Location.X {
        type=MovieSceneScriptingFloatChannel
        default=0
        thinned_from=50
        keys [[0,-12.7851],[14,-12.7851],[26,-12.9718],[35,-12.8114],[118,-12.7851],[126,-12.3649],[130,-12.2871],[134,-12.2522],[140,-12.2405],[145,-12.5448],[151,-13.0521],[159,-13.3564],[184,-13.5134],[238,-13.5391],[244,-13.4684],[249,-13.3506],[254,-12.9972],[260,-12.8647],[266,-12.815],[328,-12.7851]]
       }
       hand_r_ik_ctrl.Location.Y {
        type=MovieSceneScriptingFloatChannel
        default=0
        thinned_from=50
        keys [[0,6.1054],[14,6.1054],[26,5.867],[32,6.016],[35,6.0719],[39,6.0933],[118,6.1054],[126,5.8594],[130,5.8138],[134,5.7933],[140,5.7865],[145,5.711],[151,5.5852],[159,5.5098],[192,5.4645],[238,5.4645],[244,5.5246],[249,5.6247],[254,5.9252],[260,6.0378],[266,6.0801],[281,6.1054],[328,6.1054]]
       }
       hand_r_ik_ctrl.Location.Z {
        type=MovieSceneScriptingFloatChannel
        default=0
        thinned_from=50
        keys [[0,-19.9733],[14,-19.9733],[26,-19.7068],[35,-19.9358],[118,-19.9733],[126,-19.43],[130,-19.3294],[134,-19.2842],[140,-19.2691],[145,-19.4797],[151,-19.8306],[159,-20.0411],[218,-20.1674],[244,-20.1492],[249,-20.1189],[254,-20.0279],[328,-19.9733]]
       }
       hand_r_ik_ctrl.Rotation.X {
        type=MovieSceneScriptingFloatChannel
        default=0
        thinned_from=50
        keys [[0,-47.9791],[14,-47.9791],[26,-48.365],[35,-48.0333],[118,-47.9791],[126,-47.2746],[134,-47.0856],[140,-47.066],[145,-47.3372],[151,-47.7891],[159,-48.0603],[249,-48.162],[328,-47.9791]]
       }
       hand_r_ik_ctrl.Rotation.Y {
        type=MovieSceneScriptingFloatChannel
        default=0
        thinned_from=50
        keys [[0,-13.0308],[14,-13.0308],[26,-12.65],[32,-12.888],[35,-12.9773],[39,-13.0114],[118,-13.0308],[126,-12.1574],[130,-11.9958],[134,-11.9231],[140,-11.8989],[145,-12.2994],[151,-12.9669],[159,-13.3674],[184,-13.5739],[238,-13.6077],[244,-13.5536],[249,-13.4635],[254,-13.1931],[260,-13.0917],[266,-13.0536],[328,-13.0308]]
       }
       hand_r_ik_ctrl.Rotation.Z {
        type=MovieSceneScriptingFloatChannel
        default=0
        thinned_from=50
        keys [[0,-11.8155],[14,-11.8155],[26,-12.2022],[32,-11.9605],[35,-11.8699],[39,-11.8352],[118,-11.8155],[126,-11.1959],[130,-11.0812],[134,-11.0296],[140,-11.0124],[145,-11.3418],[151,-11.8908],[159,-12.2201],[184,-12.39],[238,-12.4178],[244,-12.3613],[249,-12.2672],[254,-11.9849],[260,-11.879],[266,-11.8393],[328,-11.8155]]
       }
       hand_r_ik_ctrl.Scale.X {type=MovieSceneScriptingFloatChannel, default=1, const [1,50]}
       hand_r_ik_ctrl.Scale.Y {type=MovieSceneScriptingFloatChannel, default=1, const [1,50]}
       hand_r_ik_ctrl.Scale.Z {type=MovieSceneScriptingFloatChannel, default=1, const [1,50]}
       Softness_2 {type=MovieSceneScriptingFloatChannel, default=0, const [0,50]}
       hand_r_ikorient_ctrl.Rotation.Roll {type=MovieSceneScriptingFloatChannel, default=0, const [0,50]}
       hand_r_ikorient_ctrl.Rotation.Pitch {type=MovieSceneScriptingFloatChannel, default=0, const [0,50]}
       hand_r_ikorient_ctrl.Rotation.Yaw {type=MovieSceneScriptingFloatChannel, default=0, const [0,50]}
       arm_r_pv_ik_ctrl.Location.X {
        type=MovieSceneScriptingFloatChannel
        default=0
        thinned_from=50
        keys [[0,-22.216],[14,-22.216],[26,-22.3351],[35,-22.2327],[118,-22.216],[126,-21.3674],[130,-21.2103],[134,-21.1397],[140,-21.1161],[145,-21.3933],[151,-21.8552],[159,-22.1324],[218,-22.2987],[249,-22.278],[328,-22.216]]
       }
       arm_r_pv_ik_ctrl.Location.Y {
        type=MovieSceneScriptingFloatChannel
        default=0
        thinned_from=50
        keys [[0,-14.1278],[14,-14.1278],[26,-14.0225],[35,-14.113],[118,-14.1278],[126,-13.9276],[134,-13.8739],[140,-13.8684],[145,-13.9767],[151,-14.1571],[159,-14.2654],[238,-14.3304],[249,-14.2798],[254,-14.1848],[266,-14.1359],[328,-14.1278]]
       }
       arm_r_pv_ik_ctrl.Location.Z {
        type=MovieSceneScriptingFloatChannel
        default=0
        thinned_from=50
        keys [[0,-0.1502],[14,-0.1502],[26,-0.0228],[32,-0.1024],[35,-0.1323],[39,-0.1437],[45,-0.1478],[53,-0.1502],[118,-0.1502],[122,-0.1225],[126,-0.0932],[130,-0.0826],[134,-0.0779],[140,-0.0763],[145,0.0788],[151,0.3373],[159,0.4923],[177,0.5505],[184,0.5723],[192,0.5854],[238,0.5854],[244,0.5164],[249,0.4015],[254,0.0567],[260,-0.0726],[266,-0.1211],[273,-0.1393],[281,-0.1502],[328,-0.1502]]
       }
       arm_r_pv_ik_ctrl.Rotation.X {
        type=MovieSceneScriptingFloatChannel
        default=0
        thinned_from=50
        keys [[0,-2.029],[14,-2.029],[26,-2.4166],[32,-2.1744],[35,-2.0835],[39,-2.0488],[45,-2.0364],[53,-2.029],[118,-2.029],[126,-1.9488],[130,-1.9339],[134,-1.9272],[140,-1.925],[145,-1.9751],[151,-2.0585],[159,-2.1085],[184,-2.1344],[238,-2.1386],[244,-2.1283],[249,-2.1112],[254,-2.0598],[260,-2.0406],[266,-2.0334],[328,-2.029]]
       }
       arm_r_pv_ik_ctrl.Rotation.Y {
        type=MovieSceneScriptingFloatChannel
        default=0
        thinned_from=50
        keys [[0,-5.9694],[14,-5.9694],[26,-6.0758],[35,-5.9844],[118,-5.9694],[126,-6.3055],[130,-6.3677],[134,-6.3957],[140,-6.405],[145,-6.1958],[151,-5.8472],[159,-5.638],[184,-5.5301],[192,-5.5125],[238,-5.5125],[244,-5.5553],[249,-5.6267],[254,-5.8409],[260,-5.9212],[266,-5.9513],[281,-5.9694],[328,-5.9694]]
       }
       arm_r_pv_ik_ctrl.Rotation.Z {
        type=MovieSceneScriptingFloatChannel
        default=0
        thinned_from=50
        keys [[0,-34.3274],[14,-34.3274],[26,-34.6041],[35,-34.3663],[118,-34.3274],[126,-33.0275],[130,-32.787],[134,-32.6788],[140,-32.6427],[145,-33.2168],[151,-34.1737],[159,-34.7478],[184,-35.0438],[238,-35.0923],[249,-34.901],[254,-34.5425],[260,-34.408],[328,-34.3274]]
       }
       arm_r_pv_ik_ctrl.Scale.X {type=MovieSceneScriptingFloatChannel, default=1, const [1,50]}
       arm_r_pv_ik_ctrl.Scale.Y {type=MovieSceneScriptingFloatChannel, default=1, const [1,50]}
       arm_r_pv_ik_ctrl.Scale.Z {type=MovieSceneScriptingFloatChannel, default=1, const [1,50]}
       body_ctrl.Location.X {
        type=MovieSceneScriptingFloatChannel
        default=0
        keys [[0,0],[7,-0.0102],[14,-0.0311],[26,-0.0508],[32,-0.0926],[35,-0.1766],[39,-0.2674],[45,-0.3185],[53,-0.3477],[60,-0.3648],[76,-0.37],[87,-0.3585],[104,-0.3162],[135,-0.8193],[141,-0.8972],[159,-0.9758],[184,-1.0025],[192,-0.9729],[197,-0.8993],[199,-0.8062],[202,-0.7038],[206,-0.5596],[210,-0.4335],[213,-0.3281],[218,-0.2687],[230,-0.2275],[238,-0.1951],[244,-0.1676],[249,-0.1358],[254,-0.0825],[260,-0.007],[266,0.0738],[273,0.1471],[281,0.1914],[289,0.208],[296,0.176],[303,0.1107],[310,0.0422],[318,0.003],[328,0]]
       }
       body_ctrl.Location.Y {type=MovieSceneScriptingFloatChannel, default=0, const [-5.5,50]}
       body_ctrl.Location.Z {
        type=MovieSceneScriptingFloatChannel
        default=0
        thinned_from=44
        keys [[0,2],[7,2.0766],[14,2.1268],[26,2.084],[35,2.0039],[39,1.9851],[45,1.9756],[53,1.9808],[60,1.991],[76,1.9978],[87,1.9986],[108,1.9921],[134,2.0519],[140,2.0544],[145,2.0494],[151,2.0341],[159,2.019],[184,2.0055],[197,2.0061],[206,2.029],[213,2.0359],[244,2.026],[249,2.0189],[260,1.9959],[266,1.9895],[273,1.9874],[310,1.9997],[328,2]]
       }
       body_ctrl.Rotation.X {
        type=MovieSceneScriptingFloatChannel
        default=0
        thinned_from=40
        keys [[0,2],[7,2.0679],[14,2.1175],[26,2.1564],[31,2.1187],[35,2.0674],[39,2.0344],[45,2.0143],[53,2.0116],[60,2.0366],[76,2.1878],[87,2.2673],[104,2.2859],[108,2.2586],[113,2.1689],[145,1.5548],[171,1.343],[192,1.3114],[197,1.3802],[199,1.6764],[202,2.0361],[206,2.4607],[210,2.6949],[213,2.8438],[218,2.9171],[230,2.9391],[238,2.9025],[244,2.7962],[249,2.6139],[254,2.404],[260,2.1664],[266,2.0514],[273,1.9999],[281,1.9876],[310,1.9993],[328,2]]
       }
       body_ctrl.Rotation.Y {
        type=MovieSceneScriptingFloatChannel
        default=0
        thinned_from=44
        keys [[0,-1.5],[14,-1.5],[26,-1.5216],[35,-1.575],[39,-1.5893],[45,-1.5944],[104,-1.5944],[122,-1.8631],[130,-1.9251],[177,-1.9251],[192,-1.9193],[197,-1.9027],[199,-1.8352],[202,-1.7005],[206,-1.4924],[210,-1.3275],[213,-1.2122],[218,-1.1397],[230,-1.1011],[238,-1.0857],[244,-1.1142],[249,-1.171],[254,-1.2638],[260,-1.3937],[266,-1.4676],[273,-1.5],[328,-1.5]]
       }
       body_ctrl.Rotation.Z {
        type=MovieSceneScriptingFloatChannel
        default=0
        thinned_from=45
        keys [[0,24.1655],[14,23.9278],[26,23.6423],[32,23.2715],[35,22.7523],[39,22.2666],[45,21.9786],[53,21.8169],[60,21.7557],[76,21.79],[87,21.9813],[112,24.2827],[116,24.4691],[126,24.5654],[159,24.5397],[177,24.4499],[197,24.3019],[206,23.8827],[213,23.6823],[218,23.6267],[244,23.5501],[249,23.5741],[254,23.6536],[260,23.8726],[266,24.1464],[273,24.309],[281,24.4043],[296,24.4592],[328,24.1655]]
       }
       body_ctrl.Scale.X {type=MovieSceneScriptingFloatChannel, default=1, const [1,50]}
       body_ctrl.Scale.Y {type=MovieSceneScriptingFloatChannel, default=1, const [1,50]}
       body_ctrl.Scale.Z {type=MovieSceneScriptingFloatChannel, default=1, const [1,50]}
       hips_ctrl.Location.X {type=MovieSceneScriptingFloatChannel, default=0, const [0,50]}
       hips_ctrl.Location.Y {type=MovieSceneScriptingFloatChannel, default=0, const [0,50]}
       hips_ctrl.Location.Z {type=MovieSceneScriptingFloatChannel, default=0, const [0,50]}
       hips_ctrl.Rotation.X {type=MovieSceneScriptingFloatChannel, default=0, const [0,50]}
       hips_ctrl.Rotation.Y {type=MovieSceneScriptingFloatChannel, default=0, const [2.5,50]}
       hips_ctrl.Rotation.Z {type=MovieSceneScriptingFloatChannel, default=0, const [0,50]}
       hips_ctrl.Scale.X {type=MovieSceneScriptingFloatChannel, default=1, const [1,50]}
       hips_ctrl.Scale.Y {type=MovieSceneScriptingFloatChannel, default=1, const [1,50]}
       hips_ctrl.Scale.Z {type=MovieSceneScriptingFloatChannel, default=1, const [1,50]}
       hips_tan_ctrl.Location.X {type=MovieSceneScriptingFloatChannel, default=0, const [0.2993,50]}
       hips_tan_ctrl.Location.Y {type=MovieSceneScriptingFloatChannel, default=0.0032, const [0.1398,50]}
       hips_tan_ctrl.Location.Z {type=MovieSceneScriptingFloatChannel, default=-0.0015, const [0.0044,50]}
       hips_tan_ctrl.Rotation.X {type=MovieSceneScriptingFloatChannel, default=0, const [0,50]}
       hips_tan_ctrl.Rotation.Y {type=MovieSceneScriptingFloatChannel, default=0, const [0,50]}
       hips_tan_ctrl.Rotation.Z {type=MovieSceneScriptingFloatChannel, default=0, const [0,50]}
       hips_tan_ctrl.Scale.X {type=MovieSceneScriptingFloatChannel, default=1, const [1,50]}
       hips_tan_ctrl.Scale.Y {type=MovieSceneScriptingFloatChannel, default=1, const [1,50]}
       hips_tan_ctrl.Scale.Z {type=MovieSceneScriptingFloatChannel, default=1, const [1,50]}
       spine_01_ctrl.Location.X {type=MovieSceneScriptingFloatChannel, default=0, const [0,50]}
       spine_01_ctrl.Location.Y {type=MovieSceneScriptingFloatChannel, default=0, const [0,50]}
       spine_01_ctrl.Location.Z {type=MovieSceneScriptingFloatChannel, default=0, const [0,50]}
       spine_01_ctrl.Rotation.X {type=MovieSceneScriptingFloatChannel, default=0, const [-7.1482,50]}
       spine_01_ctrl.Rotation.Y {type=MovieSceneScriptingFloatChannel, default=0, const [0,50]}
       spine_01_ctrl.Rotation.Z {type=MovieSceneScriptingFloatChannel, default=0, const [0,50]}
       spine_01_ctrl.Scale.X {type=MovieSceneScriptingFloatChannel, default=1, const [1,50]}
       spine_01_ctrl.Scale.Y {type=MovieSceneScriptingFloatChannel, default=1, const [1,50]}
       spine_01_ctrl.Scale.Z {type=MovieSceneScriptingFloatChannel, default=1, const [1,50]}
       spine_02_ctrl.Location.X {type=MovieSceneScriptingFloatChannel, default=0, const [0,50]}
       spine_02_ctrl.Location.Y {type=MovieSceneScriptingFloatChannel, default=0, const [0,50]}
       spine_02_ctrl.Location.Z {type=MovieSceneScriptingFloatChannel, default=0, const [0,50]}
       spine_02_ctrl.Rotation.X {type=MovieSceneScriptingFloatChannel, default=0, const [0,50]}
       spine_02_ctrl.Rotation.Y {
        type=MovieSceneScriptingFloatChannel
        default=0
        thinned_from=50
        keys [[0,0],[7,0.0594],[14,0.0811],[26,0.0604],[32,-0.0403],[35,-0.0939],[39,-0.1123],[45,-0.0828],[53,0],[60,0.058],[76,0.1043],[140,0.1043],[145,0.1769],[151,0.2208],[159,0.2386],[177,0.2495],[184,0.2545],[192,0.2569],[197,0.2657],[202,0.3402],[206,0.4197],[210,0.4929],[213,0.5549],[218,0.6039],[238,0.6596],[244,0.6686],[249,0.6729],[254,0.668],[260,0.6489],[266,0.6075],[273,0.5247],[281,0.3886],[289,0.2447],[296,0.1335],[303,0.0444],[310,0.0087],[318,0],[328,0]]
       }
       spine_02_ctrl.Rotation.Z {type=MovieSceneScriptingFloatChannel, default=0, const [0,50]}
       spine_02_ctrl.Scale.X {type=MovieSceneScriptingFloatChannel, default=1, const [1,50]}
       spine_02_ctrl.Scale.Y {type=MovieSceneScriptingFloatChannel, default=1, const [1,50]}
       spine_02_ctrl.Scale.Z {type=MovieSceneScriptingFloatChannel, default=1, const [1,50]}
       spine_03_ctrl.Location.X {type=MovieSceneScriptingFloatChannel, default=0, const [0,50]}
       spine_03_ctrl.Location.Y {type=MovieSceneScriptingFloatChannel, default=0, const [0,50]}
       spine_03_ctrl.Location.Z {type=MovieSceneScriptingFloatChannel, default=0, const [0,50]}
       spine_03_ctrl.Rotation.X {
        type=MovieSceneScriptingFloatChannel
        default=0
        thinned_from=44
        keys [[0,-5],[7,-5.4024],[14,-5.7249],[26,-5.607],[32,-5.1276],[35,-4.7497],[39,-4.4625],[45,-4.288],[53,-4.2198],[60,-4.3374],[76,-5.0541],[87,-5.618],[104,-5.9293],[122,-6.1442],[130,-6.2138],[140,-6.2683],[151,-6.2829],[159,-6.1387],[177,-5.3477],[184,-5.0722],[192,-4.909],[197,-4.8772],[199,-5.0131],[202,-5.3447],[206,-6.0421],[210,-6.7172],[213,-7.1709],[218,-7.4812],[230,-7.6586],[238,-7.5388],[244,-7.1698],[249,-6.6524],[254,-5.9609],[260,-5.0184],[266,-4.2114],[273,-3.7692],[281,-3.5722],[289,-3.5287],[296,-3.6475],[303,-3.877],[310,-4.1569],[318,-4.4949],[328,-5]]
       }
       spine_03_ctrl.Rotation.Y {type=MovieSceneScriptingFloatChannel, default=0, const [-2,50]}
       spine_03_ctrl.Rotation.Z {
        type=MovieSceneScriptingFloatChannel
        default=0
        thinned_from=43
        keys [[0,-3],[7,-3.2342],[14,-3.4213],[25,-3.6731],[30,-3.8632],[37,-4.873],[42,-5.7114],[48,-6.1466],[56,-6.3908],[76,-6.4746],[87,-6.4716],[99,-6.3157],[105,-5.806],[117,-2.3541],[125,-1.5543],[133,-1.2653],[141,-1.1258],[159,-1.0408],[177,-1.098],[184,-1.2114],[192,-1.4542],[197,-1.7875],[199,-2.2148],[202,-2.797],[206,-3.6059],[210,-4.1946],[213,-4.6589],[218,-4.9433],[230,-5.1497],[238,-5.2766],[249,-5.3757],[254,-5.3118],[260,-5.0405],[266,-4.2762],[273,-3.3941],[281,-2.7738],[289,-2.4637],[296,-2.3458],[303,-2.4303],[328,-3]]
       }
       spine_03_ctrl.Scale.X {type=MovieSceneScriptingFloatChannel, default=1, const [1,50]}
       spine_03_ctrl.Scale.Y {type=MovieSceneScriptingFloatChannel, default=1, const [1,50]}
       spine_03_ctrl.Scale.Z {type=MovieSceneScriptingFloatChannel, default=1, const [1,50]}
       spine_04_ctrl.Location.X {type=MovieSceneScriptingFloatChannel, default=0, const [0,50]}
       spine_04_ctrl.Location.Y {type=MovieSceneScriptingFloatChannel, default=0, const [0,50]}
       spine_04_ctrl.Location.Z {type=MovieSceneScriptingFloatChannel, default=0, const [0,50]}
       spine_04_ctrl.Rotation.X {type=MovieSceneScriptingFloatChannel, default=0, const [0,50]}
       spine_04_ctrl.Rotation.Y {type=MovieSceneScriptingFloatChannel, default=0, const [0,50]}
       spine_04_ctrl.Rotation.Z {type=MovieSceneScriptingFloatChannel, default=0, const [0,50]}
       spine_04_ctrl.Scale.X {type=MovieSceneScriptingFloatChannel, default=1, const [1,50]}
       spine_04_ctrl.Scale.Y {type=MovieSceneScriptingFloatChannel, default=1, const [1,50]}
       spine_04_ctrl.Scale.Z {type=MovieSceneScriptingFloatChannel, default=1, const [1,50]}
       head_ctrl.Location.X {type=MovieSceneScriptingFloatChannel, default=0, const [0,50]}
       head_ctrl.Location.Y {type=MovieSceneScriptingFloatChannel, default=0, const [0,50]}
       head_ctrl.Location.Z {type=MovieSceneScriptingFloatChannel, default=0, const [0,50]}
       head_ctrl.Rotation.X {
        type=MovieSceneScriptingFloatChannel
        default=0
        keys [[0,1.0825],[7,1.2637],[14,1.3717],[26,1.4019],[32,1.3164],[35,1.1801],[39,0.9412],[45,0.7592],[53,0.7068],[60,0.8098],[76,1.0732],[87,1.1793],[98,1.2441],[120,0.4535],[133,0.2617],[145,0.2063],[151,0.2766],[177,0.9168],[184,1.0834],[192,1.1969],[197,1.2579],[200,1.1371],[203,0.5136],[208,-1.4759],[211,-2.192],[217,-2.7621],[228,-3.0117],[236,-3.1004],[244,-3.1247],[249,-3.0565],[254,-2.507],[260,-0.5556],[263,0.7291],[269,2.4329],[278,3.0557],[289,3.2599],[296,3.2995],[303,2.9733],[310,1.9685],[318,1.3249],[328,1.0825]]
       }
       head_ctrl.Rotation.Y {
        type=MovieSceneScriptingFloatChannel
        default=0
        thinned_from=45
        keys [[0,6.5],[24,6.5],[30,7.1772],[37,8.5167],[43,8.7842],[90,8.7842],[100,6.8818],[111,-3.0947],[119,-4.1908],[184,-4.1908],[192,-4.0507],[197,-3.6184],[199,-3.0108],[202,-0.4978],[206,3.4114],[210,5.0922],[213,5.4941],[218,5.6977],[230,5.7894],[238,5.7894],[244,5.7505],[249,5.6922],[254,5.6536],[260,5.7439],[273,6.0779],[281,6.2704],[289,6.4334],[296,6.5],[328,6.5]]
       }
       head_ctrl.Rotation.Z {
        type=MovieSceneScriptingFloatChannel
        default=0
        thinned_from=40
        keys [[0,20],[7,20.2551],[14,20.4082],[26,19.29],[30,14.9285],[32,12.0726],[36,8.2164],[42,5.8947],[49,4.7567],[57,4.2602],[78,4.0212],[91,3.9478],[102,10.5658],[106,16.8264],[112,21.9751],[125,25.456],[158,27.3253],[184,25.7033],[192,24.9087],[198,23.6351],[203,20.7016],[207,15.6903],[211,12.3977],[217,10.8128],[228,9.7333],[238,9.3013],[244,9.5184],[249,9.9411],[254,10.5355],[260,12.051],[266,13.2862],[273,13.9479],[281,14.35],[289,14.6947],[296,15.188],[303,16.5253],[310,18.4361],[318,19.6545],[328,20]]
       }
       head_ctrl.Scale.X {type=MovieSceneScriptingFloatChannel, default=1, const [1,50]}
       head_ctrl.Scale.Y {type=MovieSceneScriptingFloatChannel, default=1, const [1,50]}
       head_ctrl.Scale.Z {type=MovieSceneScriptingFloatChannel, default=1, const [1,50]}
       clavicle_l_ctrl.Location.X {type=MovieSceneScriptingFloatChannel, default=0, const [0,50]}
       clavicle_l_ctrl.Location.Y {type=MovieSceneScriptingFloatChannel, default=0, const [0,50]}
       clavicle_l_ctrl.Location.Z {
        type=MovieSceneScriptingFloatChannel
        default=0
        keys [[0,0],[7,0.0504],[14,0.2015],[26,0.3145],[33,0.2059],[36,0.0568],[40,-0.1564],[46,-0.2698],[54,-0.3113],[61,-0.3323],[76,-0.2691],[87,-0.0795],[100,0.1743],[106,0.2323],[114,0.2496],[118,0.2496],[122,0.2658],[126,0.3038],[140,0.4683],[145,0.5121],[151,0.5379],[159,0.5495],[177,0.3997],[192,0.1769],[199,0.1028],[202,0.0948],[206,0.1255],[209,0.2428],[219,0.5715],[228,0.6417],[238,0.6709],[246,0.6071],[267,-0.2156],[274,-0.398],[281,-0.4913],[287,-0.5246],[296,-0.5377],[303,-0.4862],[310,-0.3427],[318,-0.1171],[328,0]]
       }
       clavicle_l_ctrl.Rotation.X {type=MovieSceneScriptingFloatChannel, default=0, const [0,50]}
       clavicle_l_ctrl.Rotation.Y {type=MovieSceneScriptingFloatChannel, default=0, const [7,50]}
       clavicle_l_ctrl.Rotation.Z {type=MovieSceneScriptingFloatChannel, default=0, const [0,50]}
       clavicle_l_ctrl.Scale.X {type=MovieSceneScriptingFloatChannel, default=1, const [1,50]}
       clavicle_l_ctrl.Scale.Y {type=MovieSceneScriptingFloatChannel, default=1, const [1,50]}
       clavicle_l_ctrl.Scale.Z {type=MovieSceneScriptingFloatChannel, default=1, const [1,50]}
       clavicle_r_ctrl.Location.X {type=MovieSceneScriptingFloatChannel, default=0, const [0,50]}
       clavicle_r_ctrl.Location.Y {type=MovieSceneScriptingFloatChannel, default=0, const [0,50]}
       clavicle_r_ctrl.Location.Z {
        type=MovieSceneScriptingFloatChannel
        default=0
        thinned_from=38
        keys [[0,-0.5239],[7,-0.4735],[14,-0.3224],[26,-0.2094],[34,-0.3159],[37,-0.4625],[41,-0.6718],[47,-0.7851],[55,-0.8267],[62,-0.8492],[76,-0.7869],[87,-0.6001],[99,-0.3496],[105,-0.2916],[114,-0.2742],[122,-0.2742],[126,-0.2478],[130,-0.2022],[140,-0.0694],[145,-0.02],[151,0.014],[159,0.0256],[177,-0.0437],[202,-0.439],[205,-0.447],[220,-0.3009],[229,-0.2019],[238,-0.1548],[246,-0.2186],[274,-0.9219],[281,-1.0152],[287,-1.0485],[296,-1.0616],[303,-1.0101],[310,-0.8665],[318,-0.641],[328,-0.5239]]
       }
       clavicle_r_ctrl.Rotation.X {type=MovieSceneScriptingFloatChannel, default=0, const [0,50]}
       clavicle_r_ctrl.Rotation.Y {type=MovieSceneScriptingFloatChannel, default=0, const [0,50]}
       clavicle_r_ctrl.Rotation.Z {type=MovieSceneScriptingFloatChannel, default=0, const [0,50]}
       clavicle_r_ctrl.Scale.X {type=MovieSceneScriptingFloatChannel, default=1, const [1,50]}
       clavicle_r_ctrl.Scale.Y {type=MovieSceneScriptingFloatChannel, default=1, const [1,50]}
       clavicle_r_ctrl.Scale.Z {type=MovieSceneScriptingFloatChannel, default=1, const [1,50]}
       neck_01_ctrl.Location.X {type=MovieSceneScriptingFloatChannel, default=0, const [0,50]}
       neck_01_ctrl.Location.Y {type=MovieSceneScriptingFloatChannel, default=0, const [0,50]}
       neck_01_ctrl.Location.Z {type=MovieSceneScriptingFloatChannel, default=0, const [0,50]}
       neck_01_ctrl.Rotation.X {type=MovieSceneScriptingFloatChannel, default=0, const [0,50]}
       neck_01_ctrl.Rotation.Y {type=MovieSceneScriptingFloatChannel, default=0, const [6,50]}
       neck_01_ctrl.Rotation.Z {type=MovieSceneScriptingFloatChannel, default=0, const [0,50]}
       neck_01_ctrl.Scale.X {type=MovieSceneScriptingFloatChannel, default=1, const [1,50]}
       neck_01_ctrl.Scale.Y {type=MovieSceneScriptingFloatChannel, default=1, const [1,50]}
       neck_01_ctrl.Scale.Z {type=MovieSceneScriptingFloatChannel, default=1, const [1,50]}
       neck_02_ctrl.Location.X {type=MovieSceneScriptingFloatChannel, default=0, const [0,50]}
       neck_02_ctrl.Location.Y {type=MovieSceneScriptingFloatChannel, default=0, const [0,50]}
       neck_02_ctrl.Location.Z {type=MovieSceneScriptingFloatChannel, default=0, const [0,50]}
       neck_02_ctrl.Rotation.X {type=MovieSceneScriptingFloatChannel, default=0, const [0,50]}
       neck_02_ctrl.Rotation.Y {type=MovieSceneScriptingFloatChannel, default=0, const [0,50]}
       neck_02_ctrl.Rotation.Z {type=MovieSceneScriptingFloatChannel, default=0, const [0,50]}
       neck_02_ctrl.Scale.X {type=MovieSceneScriptingFloatChannel, default=1, const [1,50]}
       neck_02_ctrl.Scale.Y {type=MovieSceneScriptingFloatChannel, default=1, const [1,50]}
       neck_02_ctrl.Scale.Z {type=MovieSceneScriptingFloatChannel, default=1, const [1,50]}
       head_ik_ctrl.Location.X {
        type=MovieSceneScriptingFloatChannel
        default=0
        thinned_from=50
        keys [[0,-0.9386],[14,-0.9386],[26,-0.9365],[118,-0.9386],[126,-0.9464],[140,-0.9487],[151,-0.9392],[159,-0.9357],[249,-0.9348],[266,-0.9384],[328,-0.9386]]
       }
       head_ik_ctrl.Location.Y {
        type=MovieSceneScriptingFloatChannel
        default=0
        thinned_from=50
        keys [[0,-0.3545],[14,-0.3545],[26,-0.3598],[35,-0.3552],[118,-0.3545],[126,-0.333],[130,-0.329],[134,-0.3272],[140,-0.3266],[145,-0.3361],[151,-0.3519],[159,-0.3614],[184,-0.3663],[238,-0.3671],[244,-0.3659],[249,-0.364],[254,-0.358],[260,-0.3558],[281,-0.3545],[328,-0.3545]]
       }
       head_ik_ctrl.Location.Z {
        type=MovieSceneScriptingFloatChannel
        default=0
        thinned_from=50
        keys [[0,0.0119],[14,0.0119],[26,0.0117],[118,0.0119],[126,0.0059],[130,0.0048],[134,0.0043],[140,0.0041],[145,0.0078],[151,0.0141],[159,0.0179],[184,0.0198],[192,0.0201],[238,0.0201],[244,0.0194],[249,0.0181],[254,0.0142],[260,0.0128],[266,0.0122],[281,0.0119],[328,0.0119]]
       }
       head_ik_ctrl.Rotation.X {
        type=MovieSceneScriptingFloatChannel
        default=0
        thinned_from=50
        keys [[0,0],[14,0],[26,0.1232],[32,0.0462],[35,0.0173],[39,0.0063],[45,0.0024],[53,0],[328,0]]
       }
       head_ik_ctrl.Rotation.Y {
        type=MovieSceneScriptingFloatChannel
        default=0
        thinned_from=50
        keys [[0,6.5],[118,6.5],[122,3.5227],[126,0.3742],[130,-0.7594],[134,-1.2695],[140,-1.4395],[145,-2.0843],[151,-3.1591],[159,-3.8039],[184,-4.1364],[192,-4.1908],[238,-4.1908],[244,-3.1886],[249,-1.5181],[254,3.4932],[260,5.3725],[266,6.0772],[273,6.3414],[281,6.5],[328,6.5]]
       }
       head_ik_ctrl.Rotation.Z {
        type=MovieSceneScriptingFloatChannel
        default=0
        thinned_from=50
        keys [[0,20],[14,20],[26,20.4082],[32,20.1531],[35,20.0574],[60,20],[118,20],[122,16.0625],[126,11.8986],[130,10.3995],[134,9.7249],[140,9.5],[145,15.8116],[151,26.331],[159,32.6426],[184,35.8971],[192,36.4296],[238,36.4296],[244,34.8894],[249,32.3222],[254,24.6208],[260,21.7328],[266,20.6498],[273,20.2437],[281,20],[328,20]]
       }
       head_ik_ctrl.Scale.X {type=MovieSceneScriptingFloatChannel, default=1, const [1,50]}
       head_ik_ctrl.Scale.Y {type=MovieSceneScriptingFloatChannel, default=1, const [1,50]}
       head_ik_ctrl.Scale.Z {type=MovieSceneScriptingFloatChannel, default=1, const [1,50]}
       chest_ctrl.Location.X {type=MovieSceneScriptingFloatChannel, default=0, thinned_from=50, keys [[0,0],[328,0]]}
       chest_ctrl.Location.Y {type=MovieSceneScriptingFloatChannel, default=0, const [-0.8823,50]}
       chest_ctrl.Location.Z {type=MovieSceneScriptingFloatChannel, default=0, const [-0.1199,50]}
       chest_ctrl.Rotation.X {
        type=MovieSceneScriptingFloatChannel
        default=0
        thinned_from=50
        keys [[0,-9.997],[14,-9.997],[26,-10.5009],[32,-10.186],[35,-10.0679],[39,-10.0227],[60,-9.997],[328,-9.997]]
       }
       chest_ctrl.Rotation.Y {
        type=MovieSceneScriptingFloatChannel
        default=0
        thinned_from=50
        keys [[0,-2.2538],[14,-2.2538],[26,-2.2357],[35,-2.2512],[118,-2.2538],[126,-2.3431],[130,-2.3596],[134,-2.367],[140,-2.3695],[145,-2.3275],[151,-2.2573],[159,-2.2153],[184,-2.1936],[238,-2.19],[244,-2.196],[249,-2.2059],[254,-2.2358],[260,-2.247],[328,-2.2538]]
       }
       chest_ctrl.Rotation.Z {
        type=MovieSceneScriptingFloatChannel
        default=0
        thinned_from=50
        keys [[0,-2.8144],[14,-2.8144],[26,-2.6072],[32,-2.7367],[35,-2.7853],[39,-2.8039],[53,-2.8144],[118,-2.8144],[122,-3.3117],[126,-3.8376],[130,-4.0269],[134,-4.1121],[140,-4.1405],[145,-3.6588],[151,-2.8559],[159,-2.3741],[184,-2.1257],[192,-2.0851],[238,-2.0851],[244,-2.1535],[249,-2.2674],[254,-2.6093],[260,-2.7375],[266,-2.7856],[273,-2.8036],[281,-2.8144],[328,-2.8144]]
       }
       chest_ctrl.Scale.X {type=MovieSceneScriptingFloatChannel, default=1, const [1,50]}
       chest_ctrl.Scale.Y {type=MovieSceneScriptingFloatChannel, default=1, const [1,50]}
       chest_ctrl.Scale.Z {type=MovieSceneScriptingFloatChannel, default=1, const [1,50]}
       chest_tan_ctrl.Location.X {
        type=MovieSceneScriptingFloatChannel
        default=0
        thinned_from=50
        keys [[0,0.2064],[14,0.2064],[26,0.2091],[35,0.2068],[118,0.2064],[126,0.193],[130,0.1905],[134,0.1894],[140,0.189],[145,0.1953],[151,0.2059],[159,0.2122],[184,0.2154],[238,0.216],[244,0.2151],[249,0.2136],[254,0.2091],[260,0.2074],[266,0.2068],[328,0.2064]]
       }
       chest_tan_ctrl.Location.Y {
        type=MovieSceneScriptingFloatChannel
        default=0.0049
        thinned_from=50
        keys [[0,-0.4727],[14,-0.4727],[26,-0.5346],[32,-0.4959],[35,-0.4814],[39,-0.4758],[45,-0.4738],[53,-0.4727],[145,-0.4734],[159,-0.4724],[328,-0.4727]]
       }
       chest_tan_ctrl.Location.Z {type=MovieSceneScriptingFloatChannel, default=1.2539, thinned_from=50, keys [[0,1.2405],[328,1.2405]]}
       chest_tan_ctrl.Rotation.X {
        type=MovieSceneScriptingFloatChannel
        default=0.6533
        thinned_from=50
        keys [[0,5.7534],[140,5.7534],[159,5.7866],[244,5.7884],[249,5.7823],[254,5.7642],[328,5.7534]]
       }
       chest_tan_ctrl.Rotation.Y {
        type=MovieSceneScriptingFloatChannel
        default=0
        thinned_from=50
        keys [[0,1.7282],[140,1.7282],[145,1.7003],[151,1.6539],[159,1.6261],[184,1.6118],[238,1.6094],[244,1.6205],[249,1.6391],[254,1.6948],[260,1.7156],[266,1.7235],[328,1.7282]]
       }
       chest_tan_ctrl.Rotation.Z {
        type=MovieSceneScriptingFloatChannel
        default=0
        thinned_from=50
        keys [[0,3.1643],[140,3.1643],[145,3.4751],[151,3.9931],[159,4.3039],[184,4.4642],[192,4.4904],[238,4.4904],[244,4.3661],[249,4.1589],[254,3.5373],[260,3.3042],[266,3.2168],[273,3.184],[281,3.1643],[328,3.1643]]
       }
       chest_tan_ctrl.Scale.X {type=MovieSceneScriptingFloatChannel, default=1, const [1,50]}
       chest_tan_ctrl.Scale.Y {type=MovieSceneScriptingFloatChannel, default=1, const [1,50]}
       chest_tan_ctrl.Scale.Z {type=MovieSceneScriptingFloatChannel, default=1, const [1,50]}
      ]
     }]
   }
   {
    type=MovieSceneControlRigParameterTrack
    display_name=Fortnite_Legacy_Face_CtrlRig
    sections [{
      type=MovieSceneControlRigParameterSection
      channels [
       Flipbook {type=MovieSceneScriptingBoolChannel, default=false}
       Face_Ctrl_Master.Location.X {type=MovieSceneScriptingFloatChannel, default=-24.1636}
       Face_Ctrl_Master.Location.Y {type=MovieSceneScriptingFloatChannel, default=-6.4979}
       Face_Ctrl_Master.Location.Z {type=MovieSceneScriptingFloatChannel, default=12.4749}
       Face_Ctrl_Master.Rotation.X {type=MovieSceneScriptingFloatChannel, default=0}
       Face_Ctrl_Master.Rotation.Y {type=MovieSceneScriptingFloatChannel, default=0}
       Face_Ctrl_Master.Rotation.Z {type=MovieSceneScriptingFloatChannel, default=0}
       Face_Ctrl_Master.Scale.X {type=MovieSceneScriptingFloatChannel, default=1}
       Face_Ctrl_Master.Scale.Y {type=MovieSceneScriptingFloatChannel, default=1}
       Face_Ctrl_Master.Scale.Z {type=MovieSceneScriptingFloatChannel, default=1}
       gui_follow_head {type=MovieSceneScriptingFloatChannel, default=1}
      ]
     }]
   }
  ]
 }
 mannequin2 {
  id=#3
  possessed_class=FortPlayerMannequin
  tracks [
   {type=MovieSceneSpawnTrack, sections [{type=MovieSceneSpawnSection, channels [None {type=MovieSceneScriptingBoolChannel, default=true}]}]}
   {
    type=MovieScene3DTransformTrack
    sections [{
      type=MovieScene3DTransformSection
      channels [
       Location.X {type=MovieSceneScriptingDoubleChannel, default=-90.3993}
       Location.Y {type=MovieSceneScriptingDoubleChannel, default=0}
       Location.Z {type=MovieSceneScriptingDoubleChannel, default=0}
       Rotation.X {type=MovieSceneScriptingDoubleChannel, default=0}
       Rotation.Y {type=MovieSceneScriptingDoubleChannel, default=0}
       Rotation.Z {type=MovieSceneScriptingDoubleChannel, default=77.6656}
       Scale.X {type=MovieSceneScriptingDoubleChannel, default=1}
       Scale.Y {type=MovieSceneScriptingDoubleChannel, default=1}
       Scale.Z {type=MovieSceneScriptingDoubleChannel, default=1}
      ]
     }]
   }
   {type=MovieSceneSkeletalAnimationTrack, sections [{type=MovieSceneSkeletalAnimationSection, end_frame=328, start_frame=-1}]}
  ]
 }
 mannequin4 {
  id=#4
  possessed_class=FortPlayerMannequin
  tracks [
   {type=MovieSceneSpawnTrack, sections [{type=MovieSceneSpawnSection, channels [None {type=MovieSceneScriptingBoolChannel, default=true}]}]}
   {
    type=MovieScene3DTransformTrack
    sections [{
      type=MovieScene3DTransformSection
      channels [
       Location.X {type=MovieSceneScriptingDoubleChannel, default=170.9575}
       Location.Y {type=MovieSceneScriptingDoubleChannel, default=0}
       Location.Z {type=MovieSceneScriptingDoubleChannel, default=0}
       Rotation.X {type=MovieSceneScriptingDoubleChannel, default=0}
       Rotation.Y {type=MovieSceneScriptingDoubleChannel, default=0}
       Rotation.Z {type=MovieSceneScriptingDoubleChannel, default=-180}
       Scale.X {type=MovieSceneScriptingDoubleChannel, default=1}
       Scale.Y {type=MovieSceneScriptingDoubleChannel, default=1}
       Scale.Z {type=MovieSceneScriptingDoubleChannel, default=1}
      ]
     }]
   }
   {type=MovieSceneSkeletalAnimationTrack, sections [{type=MovieSceneSkeletalAnimationSection, end_frame=328, start_frame=-1}]}
  ]
 }
]
display_rate {denominator=1, numerator=30}
tick_resolution {denominator=1, numerator=24000}
tracks []
